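# Control Rig for the F_MED_DarkNinja_01 female player body that handles corrective deformation bones.
# BeginExecution runs a single referenced DeformRig function, with all logic outside this asset and per-body curve, toggle, transform and debug inputs set here.
# Upperarm and elbow correctives are enabled, while wrist, groin and glute are switched off.
# The hierarchy also carries collar deform bones plus dyn_ strap and left-side skirt chains, which this graph does not directly reference.
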ControlRigBlueprint /Game/Characters/Player/Female/Medium/Bodies/F_MED_DarkNinja_01/Meshes/F_MED_DarkNinja_CtrlRig
graphs [
 RigVMFunctionLibrary {links [], nodes []}
 RigVMModel {nodes [
   RigUnit_BeginExecution = &RigUnit_BeginExecution()
   FN_DeformRig_FNC = RigVMFunctionReferenceNode(
    DeformRig_ Curve="(deformRig_l_upperarm_up_alpha_72_#1=False,deformRig_r_upperarm_up_alpha_73_#2=False,deformRig_l_upperarm_fr_alpha_74_#3=False,deformRig_r_upperarm_fr_alpha_75_CB51B8 ...(truncated)"
    DeformRig_Bool="(upperarm_1_#4=true,elbow_3_#5=true,wrist_7_#6=false,groin_8_#7=false,glute_17_#8=false,knee_9_40C408324CA ...(truncated)"
    DeformRig_Debug=(upperarm_debug_18_#4=False,wrist_debug_20_#6=False,lat_debug_26_#9=False,groinGlute_debug_27_#7=False)
    DeformRig_Value="(upperarm_up_18_#10=(Rotation=(X=0,Y=0,Z=0,W=1),Translation=(X=0,Y=0,Z=8),Scale3D=(X=0.7,Y=1,Z=1)),upperarm_fwd_fr_19_#11=(Rotation=(X=0,Y=0,Z=0,W=1),Translation=(X=0,Y=10,Z=0),Scale3D=(X=0.5,Y=1,Z= ...(truncated)"
    ExecuteContext=@RigUnit_BeginExecution.ExecutePin
   )
  ]}
]
hierarchy [
 root {loc=<0 0 0>, rot=<0 0 0 1>}
 pelvis {parent=root, loc=<0 -0.5301 98.0571>, rot=<0.0127 0.707 -0.0127 -0.707>}
 spine_01 {parent=pelvis, loc=<6.1208 -0.9328 0>, rot=<0 0 0.034 -0.9994>}
 spine_02 {parent=spine_01, loc=<6.5996 -0.2095 0>, rot=<0 0 0.0238 -0.9997>}
 spine_03 {parent=spine_02, loc=<6.4914 -0.1237 0>, rot=<0 0 0.0149 -0.9999>}
 spine_04 {parent=spine_03, loc=<6.9598 0.0794 0>, rot=<0 0 -0.0132 -0.9999>}
 spine_05 {parent=spine_04, loc=<6.9987 0 0>, rot=<0 0 -0.0015 -1>}
 clavicle_l {parent=spine_05, loc=<5.4831 -0.1053 -3.2946>, rot=<0.0465 -0.7289 -0.009 -0.6829>}
 upperarm_l {parent=clavicle_l, loc=<9.8658 -1.2136 -0.4105>, rot=<0.0162 -0.3249 -0.0275 -0.9452>}
 lowerarm_l {parent=upperarm_l, loc=<26.465 -0.8345 0.259>, rot=<0.0036 0.0161 -0.2298 -0.9731>}
 hand_l {parent=lowerarm_l, loc=<22.7355 -0.3203 -0.6105>, rot=<-0.6784 -0.0223 0.0735 -0.7307>}
 index_metacarpal_l {parent=hand_l, loc=<1.6535 0.0183 -2.96>, rot=<0.0158 -0.0124 -0.0173 -0.9996>}
 index_01_l {parent=index_metacarpal_l, loc=<8.3008 0.0357 -0.0524>, rot=<-0.0044 0.0485 0.0924 -0.9945>}
 index_02_l {parent=index_01_l, loc=<3.3587 -0.3004 0.0452>, rot=<-0.002 0.0027 0.3225 -0.9466>}
 index_03_l {parent=index_02_l, loc=<2.5578 -0.123 -0.0252>, rot=<-0.002 0.0067 0.2558 -0.9667>}
 middle_metacarpal_l {parent=hand_l, loc=<2.1856 -0.1725 -0.5984>, rot=<0.006 0.0168 -0.0329 -0.9993>}
 middle_01_l {parent=middle_metacarpal_l, loc=<7.3275 -0.2447 0.0129>, rot=<-0.0208 0.0042 0.0858 -0.9961>}
 middle_02_l {parent=middle_01_l, loc=<4.0123 0.1401 0.1175>, rot=<-0.0218 0.0258 0.3512 -0.9357>}
 middle_03_l {parent=middle_02_l, loc=<2.6077 0.0534 0.1373>, rot=<0.0082 0.0188 0.1708 -0.9851>}
 pinky_metacarpal_l {parent=hand_l, loc=<1.7616 -0.9753 3.3814>, rot=<-0.0397 0.0814 -0.069 -0.9935>}
 pinky_01_l {parent=pinky_metacarpal_l, loc=<7.1709 -0.4831 -0.1094>, rot=<-0.0654 -0.0035 0.2242 -0.9723>}
 pinky_02_l {parent=pinky_01_l, loc=<2.8274 0.0893 -0.1043>, rot=<0.0037 0.0011 0.2854 -0.9584>}
 pinky_03_l {parent=pinky_02_l, loc=<2.2046 0.0307 -0.0078>, rot=<-0.0002 0.0098 0.1445 -0.9895>}
 ring_metacarpal_l {parent=hand_l, loc=<2.2074 -0.3625 1.3638>, rot=<-0.0284 0.0722 -0.0458 -0.9959>}
 ring_01_l {parent=ring_metacarpal_l, loc=<7.0295 -0.6197 -0.1446>, rot=<-0.0489 -0.0252 0.1892 -0.9804>}
 ring_02_l {parent=ring_01_l, loc=<3.6174 0.4647 -0.1027>, rot=<0 0 0.2806 -0.9598>}
 ring_03_l {parent=ring_02_l, loc=<2.4152 0.114 0.0199>, rot=<0 0 0.1189 -0.9929>}
 thumb_01_l {parent=hand_l, loc=<1.3118 -0.3191 -2.459>, rot=<0.6546 -0.4042 0.0664 -0.6354>}
 thumb_02_l {parent=thumb_01_l, loc=<5.9336 -0.0448 0.0401>, rot=<-0.0037 0.0213 0.298 -0.9543>}
 thumb_03_l {parent=thumb_02_l, loc=<2.2223 0.3682 -0.0907>, rot=<0.0118 -0.013 0.0066 -0.9998>}
 deform_wrist_dn_base_l {parent=hand_l, loc=<-0.0151 -1.5022 0.064>, rot=<0.0001 0.0007 -0.0003 -1>}
 deform_wrist_dn_l {parent=deform_wrist_dn_base_l, loc=<0 0 0>, rot=<0 0 0 1>}
 deform_wrist_up_base_l {parent=hand_l, loc=<-0.0174 2.3313 0.0633>, rot=<0.0001 0.0007 -0.0003 -1>}
 deform_wrist_up_l {parent=deform_wrist_up_base_l, loc=<0 0 0>, rot=<0 0 0 1>}
 lowerarm_twist_01_l {parent=lowerarm_l, loc=<18.316 0 0>, rot=<0 0 0 1>}
 dyn_strap_l_Forearm_02 {parent=lowerarm_twist_01_l, loc=<0.267 0.7229 2.6348>, rot=<0.1086 0.4893 0.2021 -0.8414>}
 lowerarm_twist_02_l {parent=lowerarm_l, loc=<10.6707 0 0>, rot=<0 0 0 1>}
 dyn_strap_l_Forearm_01 {parent=lowerarm_twist_02_l, loc=<0.7532 -0.9255 4.1523>, rot=<0.1086 0.4893 0.2021 -0.8414>}
 deform_elbow_in_base_l {parent=lowerarm_l, loc=<0.1724 3.5755 -0.2372>, rot=<0.0002 0.0003 0.0006 -1>}
 deform_elbow_in_l {parent=deform_elbow_in_base_l, loc=<0 0 0>, rot=<0 0 0 1>}
 deform_elbow_out_base_l {parent=lowerarm_l, loc=<-0.0211 -4.0304 0.023>, rot=<0.0002 0.0003 0.0006 -1>}
 deform_elbow_out_l {parent=deform_elbow_out_base_l, loc=<0 0 0>, rot=<0 0 0 1>}
 upperarm_twist_01_l {parent=upperarm_l, loc=<7.2135 -0.1134 0>, rot=<0 0 0 1>}
 deform_upperarm_bk_base_l {parent=upperarm_twist_01_l, loc=<-6.4113 -4.3756 0.3147>, rot=<0 0.0004 0.0006 -1>}
 deform_upperarm_bk_l {parent=deform_upperarm_bk_base_l, loc=<0 0 0>, rot=<0 0 0 1>}
 deform_upperarm_fr_base_l {parent=upperarm_twist_01_l, loc=<-5.625 5.0899 0.3365>, rot=<0 0.0004 0.0006 -1>}
 deform_upperarm_fr_l {parent=deform_upperarm_fr_base_l, loc=<0 0 0>, rot=<0 0 0 1>}
 deform_upperarm_up_base_l {parent=upperarm_twist_01_l, loc=<-8.8937 0.4847 3.9503>, rot=<0 0.0004 0.0006 -1>}
 deform_upperarm_up_l {parent=deform_upperarm_up_base_l, loc=<0 0 0>, rot=<0 0 0 1>}
 upperarm_twist_02_l {parent=upperarm_l, loc=<19.7998 -0.1134 0>, rot=<0 0 0 1>}
 clavicle_r {parent=spine_05, loc=<5.4831 -0.1053 3.2946>, rot=<0.7289 0.0465 -0.6829 0.009>}
 upperarm_r {parent=clavicle_r, loc=<-9.8658 1.2136 0.4104>, rot=<0.0162 -0.3249 -0.0275 -0.9452>}
 lowerarm_r {parent=upperarm_r, loc=<-26.4651 0.8345 -0.259>, rot=<0.0036 0.0161 -0.2298 -0.9731>}
 hand_r {parent=lowerarm_r, loc=<-22.7356 0.3203 0.6106>, rot=<-0.6784 -0.0223 0.0735 -0.7307>}
 index_metacarpal_r {parent=hand_r, loc=<-1.6534 -0.0181 2.9601>, rot=<0.0158 -0.0124 -0.0173 -0.9996>}
 index_01_r {parent=index_metacarpal_r, loc=<-8.3008 -0.0357 0.0524>, rot=<-0.0044 0.0485 0.0924 -0.9945>}
 index_02_r {parent=index_01_r, loc=<-3.3589 0.3003 -0.0453>, rot=<-0.002 0.0027 0.3225 -0.9466>}
 index_03_r {parent=index_02_r, loc=<-2.5578 0.123 0.0252>, rot=<-0.002 0.0067 0.2558 -0.9667>}
 middle_metacarpal_r {parent=hand_r, loc=<-2.1856 0.1726 0.5984>, rot=<0.006 0.0168 -0.0329 -0.9993>}
 middle_01_r {parent=middle_metacarpal_r, loc=<-7.3273 0.2447 -0.0129>, rot=<-0.0208 0.0042 0.0858 -0.9961>}
 middle_02_r {parent=middle_01_r, loc=<-4.0124 -0.14 -0.1175>, rot=<-0.0218 0.0258 0.3512 -0.9357>}
 middle_03_r {parent=middle_02_r, loc=<-2.6077 -0.0533 -0.1373>, rot=<0.0082 0.0188 0.1708 -0.9851>}
 pinky_metacarpal_r {parent=hand_r, loc=<-1.7617 0.9753 -3.3814>, rot=<-0.0397 0.0814 -0.069 -0.9935>}
 pinky_01_r {parent=pinky_metacarpal_r, loc=<-7.1707 0.4832 0.1094>, rot=<-0.0654 -0.0035 0.2242 -0.9723>}
 pinky_02_r {parent=pinky_01_r, loc=<-2.8274 -0.0892 0.1043>, rot=<0.0037 0.0011 0.2854 -0.9584>}
 pinky_03_r {parent=pinky_02_r, loc=<-2.2045 -0.031 0.0078>, rot=<-0.0002 0.0098 0.1445 -0.9895>}
 ring_metacarpal_r {parent=hand_r, loc=<-2.2073 0.3627 -1.3638>, rot=<-0.0284 0.0722 -0.0458 -0.9959>}
 ring_01_r {parent=ring_metacarpal_r, loc=<-7.0297 0.6197 0.1446>, rot=<-0.0489 -0.0252 0.1892 -0.9804>}
 ring_02_r {parent=ring_01_r, loc=<-3.6171 -0.4648 0.1027>, rot=<0 0 0.2806 -0.9598>}
 ring_03_r {parent=ring_02_r, loc=<-2.4153 -0.1141 -0.0199>, rot=<0 0 0.1189 -0.9929>}
 thumb_01_r {parent=hand_r, loc=<-1.3119 0.3192 2.459>, rot=<0.6546 -0.4042 0.0664 -0.6354>}
 thumb_02_r {parent=thumb_01_r, loc=<-5.9335 0.0447 -0.0401>, rot=<-0.0037 0.0213 0.298 -0.9543>}
 thumb_03_r {parent=thumb_02_r, loc=<-2.2223 -0.3682 0.0907>, rot=<0.0118 -0.013 0.0066 -0.9998>}
 weapon_r {parent=hand_r, loc=<0.1975 3.4112 -0.3811>, rot=<0 0 0 -1>}
 deform_wrist_dn_base_r {parent=hand_r, loc=<0.015 1.5024 -0.064>, rot=<0.0001 0.0007 -0.0003 -1>}
 deform_wrist_dn_r {parent=deform_wrist_dn_base_r, loc=<0 0 0>, rot=<0 0 0 1>}
 deform_wrist_up_base_r {parent=hand_r, loc=<0.0173 -2.3311 -0.0634>, rot=<0.0001 0.0007 -0.0003 -1>}
 deform_wrist_up_r {parent=deform_wrist_up_base_r, loc=<0 0 0>, rot=<0 0 0 1>}
 lowerarm_twist_01_r {parent=lowerarm_r, loc=<-18.3161 0 0>, rot=<0 0 0 1>}
 dyn_strap_r_Forearm_02 {parent=lowerarm_twist_01_r, loc=<0.3102 0.3988 -2.97>, rot=<0.8414 -0.2021 0.4893 0.1086>}
 lowerarm_twist_02_r {parent=lowerarm_r, loc=<-10.6707 0 0>, rot=<0 0 0 1>}
 dyn_strap_r_Forearm_01 {parent=lowerarm_twist_02_r, loc=<-0.7531 0.9256 -4.1523>, rot=<0.8414 -0.2021 0.4893 0.1086>}
 deform_elbow_in_base_r {parent=lowerarm_r, loc=<-0.1724 -3.5755 0.2372>, rot=<0.0002 0.0003 0.0006 -1>}
 deform_elbow_in_r {parent=deform_elbow_in_base_r, loc=<0 0 0>, rot=<0 0 0 1>}
 deform_elbow_out_base_r {parent=lowerarm_r, loc=<0.0211 4.0304 -0.023>, rot=<0.0002 0.0003 0.0006 -1>}
 deform_elbow_out_r {parent=deform_elbow_out_base_r, loc=<0 0 0>, rot=<0 0 0 1>}
 upperarm_twist_01_r {parent=upperarm_r, loc=<-7.2135 0.1134 0>, rot=<0 0 0 1>}
 deform_upperarm_bk_base_r {parent=upperarm_twist_01_r, loc=<6.4113 4.3756 -0.3146>, rot=<0 0.0004 0.0006 -1>}
 deform_upperarm_bk_r {parent=deform_upperarm_bk_base_r, loc=<0 0 0>, rot=<0 0 0 1>}
 deform_upperarm_fr_base_r {parent=upperarm_twist_01_r, loc=<5.6249 -5.0899 -0.3365>, rot=<0 0.0004 0.0006 -1>}
 deform_upperarm_fr_r {parent=deform_upperarm_fr_base_r, loc=<0 0 0>, rot=<0 0 0 1>}
 deform_upperarm_up_base_r {parent=upperarm_twist_01_r, loc=<8.8936 -0.4847 -3.9502>, rot=<0 0.0004 0.0006 -1>}
 deform_upperarm_up_r {parent=deform_upperarm_up_base_r, loc=<0 0 0>, rot=<0 0 0 1>}
 upperarm_twist_02_r {parent=upperarm_r, loc=<-19.7998 0.1134 0>, rot=<0 0 0 1>}
 neck_01 {parent=spine_05, loc=<11.3138 0.1529 0>, rot=<0 0 -0.2003 -0.9797>}
 neck_02 {parent=neck_01, loc=<4.5627 0 0>, rot=<0 0 0.0504 -0.9987>}
 head {parent=neck_02, loc=<5.3914 0 0>, rot=<0 0 0.115 -0.9934>}
 deform_collar_front_r_01 {parent=spine_05, loc=<9.8943 5.1679 8.8312>, rot=<0.0641 0.352 -0.3885 -0.8492>}
 deform_collar_back_r_01 {parent=spine_05, loc=<15.564 -2.5076 9.1051>, rot=<-0.247 0.2448 -0.1001 -0.9322>}
 deform_collar_back_01 {parent=spine_05, loc=<17.4834 -7.7267 0.0991>, rot=<0.0019 -0.0003 0.1786 -0.9839>}
 deform_collar_back_l_01 {parent=spine_05, loc=<15.564 -2.5076 -9.1051>, rot=<0.9322 0.1001 -0.2448 0.247>}
 deform_collar_front_l_01 {parent=spine_05, loc=<9.8947 5.1679 -8.8312>, rot=<0.8492 0.3885 -0.352 -0.0641>}
 thigh_half_l {parent=pelvis, loc=<-1.0855 -0.2109 -8.7053>, rot=<0.0009 0.0355 0.0155 -0.9993>}
 thigh_l {parent=pelvis, loc=<-1.0855 -0.2109 -8.7053>, rot=<0.0009 0.0355 0.0155 -0.9993>}
 calf_l {parent=thigh_l, loc=<-42.5284 -0.0019 -0.7297>, rot=<0.0006 -0.0161 -0.0391 -0.9991>}
 calf_twist_01_l {parent=calf_l, loc=<-35.8483 -1.9197 -0.5794>, rot=<0.0237 0.0095 -0.0035 -0.9997>}
 calf_twist_02_l {parent=calf_l, loc=<-19.1546 -1.7943 -0.2638>, rot=<0.0237 0.0095 -0.0035 -0.9997>}
 foot_l {parent=calf_l, loc=<-46.7402 -2.5014 -0.592>, rot=<0.0427 -0.0238 0.0353 -0.9982>}
 ball_l {parent=foot_l, loc=<-3.5658 10.8066 -0.1544>, rot=<0.0001 -0.0001 -0.7063 -0.7079>}
 deform_knee_in_base_l {parent=calf_l, loc=<0.8031 -5.0231 0.2456>, rot=<0 0 0 -1>}
 deform_knee_in_l {parent=deform_knee_in_base_l, loc=<0 0 0>, rot=<0 0 0 1>}
 deform_knee_out_base_l {parent=calf_l, loc=<0 4 0>, rot=<0 0 0 1>}
 deform_knee_out_l {parent=deform_knee_out_base_l, loc=<0 0 0>, rot=<0 0 0 1>}
 thigh_twist_01_l {parent=thigh_l, loc=<-27.952 0 0>, rot=<0.0013 0 0 -1>}
 dyn_SkirtThigh_01_MSTR {parent=thigh_twist_01_l, loc=<17.8481 0 0>, rot=<0.0155 -0.9999 0 0>}
 dyn_skirtSIde_01_Base {parent=dyn_SkirtThigh_01_MSTR, loc=<-1.0805 -4.1428 7.5928>, rot=<0.6012 -0.0231 0.0087 -0.7987>}
 dyn_skirtSIde_01 {parent=dyn_skirtSIde_01_Base, loc=<0 0 0>, rot=<0 0 0 1>}
 dyn_skirtSIde_02 {parent=dyn_skirtSIde_01, loc=<12.8439 0 0>, rot=<0.0152 0.0009 0.0175 -0.9997>}
 dyn_skirtSIde_03 {parent=dyn_skirtSIde_02, loc=<12.824 0 0>, rot=<0.0044 0.0007 0.0154 -0.9999>}
 dyn_skirtSIde_04 {parent=dyn_skirtSIde_03, loc=<16.3269 0 0>, rot=<0.0013 0.0073 -0.0092 -0.9999>}
 dyn_skirtSIde_05 {parent=dyn_skirtSIde_04, loc=<12.8039 0 0>, rot=<0.0015 0.0092 0.0323 -0.9994>}
 dyn_skirtFrontSIde_01_Base {parent=dyn_SkirtThigh_01_MSTR, loc=<0.1728 3.7253 6.065>, rot=<0.8714 -0.0411 0.0708 -0.4837>}
 dyn_skirtFrontSIde_01 {parent=dyn_skirtFrontSIde_01_Base, loc=<0 0 0>, rot=<0 0 0 1>}
 dyn_skirtFrontSIde_02 {parent=dyn_skirtFrontSIde_01, loc=<12.7464 0 0>, rot=<0.0969 0.0328 0.0098 -0.9947>}
 dyn_skirtFrontSIde_03 {parent=dyn_skirtFrontSIde_02, loc=<12.7117 0 0>, rot=<-0.1022 0.0056 0.0177 -0.9946>}
 dyn_skirtFrontSIde_04 {parent=dyn_skirtFrontSIde_03, loc=<15.9402 0 0>, rot=<0.0055 0.0168 -0.0195 -0.9997>}
 dyn_skirtFrontSIde_05 {parent=dyn_skirtFrontSIde_04, loc=<12.7735 0 0>, rot=<-0.0284 -0.0186 0.0265 -0.9991>}
 dyn_skirtFront_01_Base {parent=dyn_SkirtThigh_01_MSTR, loc=<1.8215 7.9401 -0.6225>, rot=<0.9738 -0.0101 0.0744 -0.2145>}
 dyn_skirtFront_01 {parent=dyn_skirtFront_01_Base, loc=<0 0 0>, rot=<0 0 0 1>}
 dyn_skirtFront_02 {parent=dyn_skirtFront_01, loc=<11.4677 0 0>, rot=<-0.1223 0.0038 -0.0273 -0.9921>}
 dyn_skirtFront_03 {parent=dyn_skirtFront_02, loc=<12.3267 0 0>, rot=<-0.0003 0.028 -0.005 -0.9996>}
 dyn_skirtFront_04 {parent=dyn_skirtFront_03, loc=<16.242 0 0>, rot=<0.0231 0.008 0.0208 -0.9995>}
 dyn_skirtFront_05 {parent=dyn_skirtFront_04, loc=<12.813 0 0>, rot=<0.012 0.0085 0.0401 -0.9991>}
 deform_groin_base_l {parent=thigh_l, loc=<0.4801 5.6776 -0.2486>, rot=<0 0 0 -1>}
 deform_groin_l {parent=deform_groin_base_l, loc=<0 0 0>, rot=<0 0 0 1>}
 thigh_r {parent=pelvis, loc=<-1.0853 -0.2109 8.7053>, rot=<0.0355 -0.0009 0.9993 0.0155>}
 calf_r {parent=thigh_r, loc=<42.5286 0.0019 0.7297>, rot=<0.0006 -0.0161 -0.0391 -0.9991>}
 calf_twist_01_r {parent=calf_r, loc=<35.8485 1.9197 0.5794>, rot=<0.0237 0.0095 -0.0035 -0.9997>}
 calf_twist_02_r {parent=calf_r, loc=<19.1546 1.7943 0.2638>, rot=<0.0237 0.0095 -0.0035 -0.9997>}
 dyn_strap_r_Calf_01 {parent=calf_twist_02_r, loc=<-10.8185 0.3588 5.4953>, rot=<-0.6847 0.0133 0.727 0.0496>}
 dyn_strap_r_Calf_02 {parent=calf_twist_02_r, loc=<-1.9124 0.4009 5.9902>, rot=<-0.6847 0.0133 0.727 0.0496>}
 dyn_strap_r_Calf_03 {parent=calf_twist_02_r, loc=<6.279 0.7937 4.6318>, rot=<-0.6847 0.0133 0.727 0.0496>}
 foot_r {parent=calf_r, loc=<46.7403 2.5014 0.592>, rot=<0.0427 -0.0238 0.0353 -0.9982>}
 ball_r {parent=foot_r, loc=<3.5658 -10.8066 0.1544>, rot=<0.0001 -0.0001 -0.7063 -0.7079>}
 deform_knee_in_base_r {parent=calf_r, loc=<-0.8031 5.0231 -0.2457>, rot=<0 0 0 -1>}
 deform_knee_in_r {parent=deform_knee_in_base_r, loc=<0 0 0>, rot=<0 0 0 1>}
 deform_knee_out_base_r {parent=calf_r, loc=<0 -4 0>, rot=<0 0 0 1>}
 deform_knee_out_r {parent=deform_knee_out_base_r, loc=<0 0 0>, rot=<0 0 0 1>}
 thigh_twist_01_r {parent=thigh_r, loc=<27.952 0 0>, rot=<0.0013 0 0 -1>}
 deform_groin_base_r {parent=thigh_r, loc=<-0.4799 -5.6776 0.2485>, rot=<0 0 0 -1>}
 deform_groin_r {parent=deform_groin_base_r, loc=<0 0 0>, rot=<0 0 0 1>}
 deform_glute_base_l {parent=pelvis, loc=<-7.3466 -9.0213 -9.1245>, rot=<0.0172 0.031 -0.4477 -0.8935>}
 deform_glute_l {parent=deform_glute_base_l, loc=<0 0 0>, rot=<0 0 0 1>}
 deform_glute_base_r {parent=pelvis, loc=<-7.3464 -9.0213 9.1245>, rot=<0.031 -0.0172 0.8935 -0.4477>}
 deform_glute_r {parent=deform_glute_base_r, loc=<0 0 0>, rot=<0 0 0 1>}
 dyn_SkirtPelvis_01_MSTR {parent=pelvis, loc=<-11.159 0.1012 -9.4218>, rot=<0 -0.9994 0.0027 -0.0354>}
 dyn_skirtBackSIde_01_Base {parent=dyn_SkirtPelvis_01_MSTR, loc=<-1.9429 -10.7562 2.9304>, rot=<0.3279 0.0196 0.024 -0.9442>}
 dyn_skirtBackSIde_01 {parent=dyn_skirtBackSIde_01_Base, loc=<0 0 0>, rot=<0 0 0 1>}
 dyn_skirtBackSIde_02 {parent=dyn_skirtBackSIde_01, loc=<12.923 0 0>, rot=<0.0738 0.0058 -0.008 -0.9972>}
 dyn_skirtBackSIde_03 {parent=dyn_skirtBackSIde_02, loc=<13.0135 0 0>, rot=<-0.0648 -0.002 0.012 -0.9978>}
 dyn_skirtBackSIde_04 {parent=dyn_skirtBackSIde_03, loc=<16.4122 0 0>, rot=<0.048 -0.0012 -0.0126 -0.9988>}
 dyn_skirtBackSIde_05 {parent=dyn_skirtBackSIde_04, loc=<12.7875 0 0>, rot=<-0.0042 -0.029 0.0147 -0.9995>}
 dyn_skirtBack_01_Base {parent=dyn_SkirtPelvis_01_MSTR, loc=<-2.1861 -13.783 -4.3136>, rot=<0.1277 0.0388 0.0405 -0.9902>}
 dyn_skirtBack_01 {parent=dyn_skirtBack_01_Base, loc=<0 0 0>, rot=<0 0 0 1>}
 dyn_skirtBack_02 {parent=dyn_skirtBack_01, loc=<12.9948 0 0>, rot=<0.0023 -0.0088 -0.0412 -0.9991>}
 dyn_skirtBack_03 {parent=dyn_skirtBack_02, loc=<12.8364 0 0>, rot=<-0.0068 0.0173 0.0386 -0.9991>}
 dyn_skirtBack_04 {parent=dyn_skirtBack_03, loc=<11.3961 0 0>, rot=<0.0098 -0.0444 -0.0326 -0.9984>}
 dyn_skirtBack_05 {parent=dyn_skirtBack_04, loc=<14.4202 0 0>, rot=<0 0 0 1>}
]
