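# Physics asset for the female medium Dark Ninja body that defines simulation of a long skirt.
# Five chains (front, front-side, side, back-side, back) use box bodies with limited swing and twist between segments.
# Angular-free constraints cross-link neighboring chains, except back, and tie three chain bases to the left thigh twist.
# Mostly left-leg capsules, a right-calf capsule and a wide, flat root box supply body and floor shapes.
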PhysicsAsset /Game/Characters/Player/Female/Medium/Bodies/F_MED_DarkNinja_01/Meshes/F_MED_DarkNinja_Physics
bodies {
 calf_l [
  {shapeType=Sphere, radius=7.4775, center=<-16.254 -4.53 0.423>}
  {shapeType=Capsule, length=32.6394, radius=9.3058, center=<-19.8458 -3.4157 1.6101>, rotation=<rot 87.4893 -179.9323 -179.9999>}
 ]
 calf_r [{shapeType=Capsule, length=49.7224, radius=7.5523, shapeName=calf_r_capsule, center=<26.269 3.4045 0.2712>, rotation=<rot 0 89.2395 -89.9263>}]
 calf_twist_02_l [{shapeType=Capsule, length=26.7032, radius=7.1087, center=<-6.9135 -0.9959 0.5088>, rotation=<rot 89.0539 -174.2442 -173.1056>}]
 dyn_skirtBackSIde_01 [{shapeType=Box, extentX=2.7342, extentY=14.3178, extentZ=15.7551, center=<6.5067 0.0005 0.6137>, rotation=<rot 0 89.9999 -4.0005>}]
 dyn_skirtBackSIde_01_Base [{shapeType=Box, extentX=1.01, extentY=1.01, extentZ=1.01, rotation=<rot 0 -90 0>}]
 dyn_skirtBackSIde_02 [{shapeType=Box, extentX=2.7342, extentY=14.3178, extentZ=22.2876, center=<6.5008 -2.0447 0.6049>, rotation=<rot 7.8843 90.1606 1.1722>}]
 dyn_skirtBackSIde_03 [{shapeType=Box, extentX=4.7342, extentY=19.6479, extentZ=38.6613, center=<7.9105 -1.7798 -0.13>, rotation=<rot -0.0001 89.9998 -0.0001>}]
 dyn_skirtBackSIde_04 [{shapeType=Box, extentX=4.7342, extentY=17.09, extentZ=39.6542, center=<6.5005 -1.024 0.6122>, rotation=<rot 0 89.9998 1.6055>}]
 dyn_skirtBackSIde_05 [{shapeType=Box, extentX=4.7342, extentY=14.3178, extentZ=34.0555, center=<6.5067 -1.227 0.6136>, rotation=<rot 0 89.9999 0.7142>}]
 dyn_skirtBack_01 [{shapeType=Box, extentX=2.7342, extentY=14.3178, extentZ=15.7551, center=<6.5067 0.0005 0.6137>, rotation=<rot 0 89.9999 -4.0005>}]
 dyn_skirtBack_01_Base [{shapeType=Box, extentX=1.01, extentY=1.01, extentZ=1.01, rotation=<rot 0 -90 0>}]
 dyn_skirtBack_02 [{shapeType=Box, extentX=2.7342, extentY=14.3178, extentZ=16.8311, center=<6.4638 0.0005 0>, rotation=<rot 0 89.9999 -3.382>}]
 dyn_skirtBack_03 [{shapeType=Box, extentX=2.7342, extentY=17.09, extentZ=30.4942, center=<6.4638 0.0005 0>, rotation=<rot 0 89.9999 -3.1301>}]
 dyn_skirtBack_04 [{shapeType=Box, extentX=2.7342, extentY=17.09, extentZ=32.2252, center=<6.4638 0.0005 0>, rotation=<rot 0 90 0>}]
 dyn_skirtBack_05 [{shapeType=Box, extentX=2.7342, extentY=8.5269, extentZ=29.1476, center=<3.3469 0.0003 -0.3277>, rotation=<rot 0 90 0>}]
 dyn_skirtFrontSIde_01 [{shapeType=Box, extentX=2.7342, extentY=14.3178, extentZ=15.7551, center=<6.5067 0.0005 0.6137>, rotation=<rot 0 89.9998 1.1612>}]
 dyn_skirtFrontSIde_01_Base [{shapeType=Box, extentX=2.7342, extentY=1.8456, extentZ=9.6675, center=<0.0171 0.0003 0.0212>, rotation=<rot 0 89.9998 1.1612>}]
 dyn_skirtFrontSIde_02 [{shapeType=Box, extentX=2.7342, extentY=14.3178, extentZ=16.8311, center=<6.5067 0.0005 0.6137>, rotation=<rot 0 89.9998 1.1612>}]
 dyn_skirtFrontSIde_03 [{shapeType=Box, extentX=2.7342, extentY=17.09, extentZ=33.3504, center=<6.5067 0.0005 0.6137>, rotation=<rot 0 89.9998 1.1612>}]
 dyn_skirtFrontSIde_04 [{shapeType=Box, extentX=2.7342, extentY=17.09, extentZ=33.3504, center=<6.5067 0.0005 0.6137>, rotation=<rot 0 89.9998 1.1612>}]
 dyn_skirtFrontSIde_05 [{shapeType=Box, extentX=3.6226, extentY=14.3178, extentZ=31.4112, center=<6.5067 0.0005 0.6137>, rotation=<rot 0 89.9998 1.1612>}]
 dyn_skirtFront_01 [{shapeType=Box, extentX=2.7342, extentY=14.3178, extentZ=15.7551, center=<6.5067 0.0005 0.6137>, rotation=<rot 0 89.9998 1.1612>}]
 dyn_skirtFront_01_Base [{shapeType=Box, extentX=2.7342, extentY=2.2181, extentZ=9.6672, rotation=<rot 0 89.9998 1.1612>}]
 dyn_skirtFront_02 [{shapeType=Box, extentX=2.7342, extentY=14.3178, extentZ=16.8311, center=<6.5067 0.0005 0.6137>, rotation=<rot 0 89.9998 1.1612>}]
 dyn_skirtFront_03 [{shapeType=Box, extentX=3.9842, extentY=17.09, extentZ=33.6004, center=<6.4642 -1.0098 0.6601>, rotation=<rot 0 89.9998 1.1612>}]
 dyn_skirtFront_04 [{shapeType=Box, extentX=3.9842, extentY=17.09, extentZ=33.6004, center=<6.5067 -1.0118 0.6137>, rotation=<rot 0 89.9998 1.1612>}]
 dyn_skirtFront_05 [{shapeType=Box, extentX=3.9842, extentY=14.3178, extentZ=32.6877, center=<6.5067 -0.9266 0.6137>, rotation=<rot 0 89.9998 1.1612>}]
 dyn_skirtSIde_01 [{shapeType=Box, extentX=2.7342, extentY=14.3178, extentZ=15.7551, center=<6.5067 0.0005 0.6137>, rotation=<rot 0 89.9998 1.1612>}]
 dyn_skirtSIde_01_Base [{shapeType=Box, extentX=2.7342, extentY=2.6282, extentZ=9.7157, center=<0.2502 0.0004 0.4869>, rotation=<rot 0 89.9998 1.1612>}]
 dyn_skirtSIde_02 [{shapeType=Box, extentX=2.7342, extentY=14.3178, extentZ=16.8311, center=<6.5067 -0.8513 0.6137>, rotation=<rot 0 89.9998 1.1612>}]
 dyn_skirtSIde_03 [{shapeType=Box, extentX=3.3485, extentY=17.09, extentZ=29.3504, center=<6.5067 -0.7634 0.6137>, rotation=<rot 0 89.9998 1.1612>}]
 dyn_skirtSIde_04 [{shapeType=Box, extentX=3.3485, extentY=17.09, extentZ=29.3504, center=<6.4926 -0.7633 0.6119>, rotation=<rot 0 89.9998 1.1612>}]
 dyn_skirtSIde_05 [{shapeType=Box, extentX=3.3485, extentY=14.3178, extentZ=26.7551, center=<6.5418 -0.7626 0.6094>, rotation=<rot 0 89.9998 1.1612>}]
 foot_l [
  {shapeType=Capsule, length=13.1056, radius=8.612, center=<-3.1312 5.0329 -0.3437>, rotation=<rot 0 0 -90>}
  {shapeType=Capsule, length=7.9662, radius=5.0298, center=<-0.3291 0.0002 0.3915>, rotation=<rot -85.5413 -0.0001 0.0001>}
 ]
 root [{shapeType=Box, extentX=226.01, extentY=202.51, extentZ=19.51, center=<0 0 -4.0103>, rotation=<rot 0 -90 0>}]
 thigh_twist_01_l [{shapeType=Capsule, length=28.1969, radius=9.2536, center=<2.1261 -2.913 2.3159>, rotation=<rot 84.8702 180 180>}]
}
constraints [
 {bone1Name=dyn_skirtBack_01, bone2Name=dyn_skirtBack_01_Base, swing1LimitDegrees=45, swing2LimitDegrees=45, twistLimitDegrees=8}
 {bone1Name=dyn_skirtBack_02, bone2Name=dyn_skirtBack_01, swing1LimitDegrees=12, swing2LimitDegrees=18, twistLimitDegrees=8}
 {bone1Name=dyn_skirtBack_03, bone2Name=dyn_skirtBack_02, swing1LimitDegrees=12, swing2LimitDegrees=30, twistLimitDegrees=8}
 {bone1Name=dyn_skirtBack_04, bone2Name=dyn_skirtBack_03, swing1LimitDegrees=12, swing2LimitDegrees=30, twistLimitDegrees=8}
 {bone1Name=dyn_skirtBack_05, bone2Name=dyn_skirtBack_04, swing1LimitDegrees=12, swing2LimitDegrees=30, twistLimitDegrees=8}
 {bone1Name=dyn_skirtBackSIde_01, bone2Name=dyn_skirtBackSIde_01_Base, swing1LimitDegrees=45, swing2LimitDegrees=45, twistLimitDegrees=8}
 {bone1Name=dyn_skirtBackSIde_02, bone2Name=dyn_skirtBackSIde_01, swing1LimitDegrees=30, swing2LimitDegrees=30, twistLimitDegrees=20}
 {bone1Name=dyn_skirtBackSIde_03, bone2Name=dyn_skirtBackSIde_02, swing1LimitDegrees=30, swing2LimitDegrees=30, twistLimitDegrees=20}
 {bone1Name=dyn_skirtBackSIde_04, bone2Name=dyn_skirtBackSIde_03, swing1LimitDegrees=30, swing2LimitDegrees=30, twistLimitDegrees=20}
 {bone1Name=dyn_skirtBackSIde_05, bone2Name=dyn_skirtBackSIde_04, swing1LimitDegrees=30, swing2LimitDegrees=30, twistLimitDegrees=20}
 {bone1Name=dyn_skirtSIde_01, bone2Name=dyn_skirtSIde_01_Base, swing1LimitDegrees=45, swing2LimitDegrees=45, twistLimitDegrees=15}
 {bone1Name=dyn_skirtSIde_02, bone2Name=dyn_skirtSIde_01, swing1LimitDegrees=15, swing2LimitDegrees=35, twistLimitDegrees=15}
 {bone1Name=dyn_skirtSIde_03, bone2Name=dyn_skirtSIde_02, swing1LimitDegrees=15, swing2LimitDegrees=35, twistLimitDegrees=15}
 {bone1Name=dyn_skirtSIde_04, bone2Name=dyn_skirtSIde_03, swing1LimitDegrees=15, swing2LimitDegrees=35, twistLimitDegrees=15}
 {bone1Name=dyn_skirtSIde_05, bone2Name=dyn_skirtSIde_04, swing1LimitDegrees=15, swing2LimitDegrees=35, twistLimitDegrees=15}
 {bone1Name=dyn_skirtFrontSIde_01, bone2Name=dyn_skirtFrontSIde_01_Base, swing1LimitDegrees=45, swing2LimitDegrees=45, twistLimitDegrees=15}
 {bone1Name=dyn_skirtFrontSIde_02, bone2Name=dyn_skirtFrontSIde_01, swing1LimitDegrees=12, swing2LimitDegrees=30, twistLimitDegrees=15}
 {bone1Name=dyn_skirtFrontSIde_03, bone2Name=dyn_skirtFrontSIde_02, swing1LimitDegrees=12, swing2LimitDegrees=30, twistLimitDegrees=15}
 {bone1Name=dyn_skirtFrontSIde_04, bone2Name=dyn_skirtFrontSIde_03, swing1LimitDegrees=12, swing2LimitDegrees=30, twistLimitDegrees=15}
 {bone1Name=dyn_skirtFrontSIde_05, bone2Name=dyn_skirtFrontSIde_04, swing1LimitDegrees=12, swing2LimitDegrees=30, twistLimitDegrees=15}
 {bone1Name=dyn_skirtFront_01, bone2Name=dyn_skirtFront_01_Base, swing1LimitDegrees=45, swing2LimitDegrees=45, twistLimitDegrees=15}
 {bone1Name=dyn_skirtFront_02, bone2Name=dyn_skirtFront_01, swing1LimitDegrees=20, swing2LimitDegrees=20, twistLimitDegrees=15}
 {bone1Name=dyn_skirtFront_03, bone2Name=dyn_skirtFront_02, swing1LimitDegrees=20, swing2LimitDegrees=20, twistLimitDegrees=15}
 {bone1Name=dyn_skirtFront_04, bone2Name=dyn_skirtFront_03, swing1LimitDegrees=20, swing2LimitDegrees=20, twistLimitDegrees=15}
 {bone1Name=dyn_skirtFront_05, bone2Name=dyn_skirtFront_04, swing1LimitDegrees=20, swing2LimitDegrees=20, twistLimitDegrees=15}
 {
  bone1Name=dyn_skirtFrontSIde_04
  bone2Name=dyn_skirtFront_04
  swing1LimitDegrees=45
  swing1Motion=Free
  swing2LimitDegrees=45
  swing2Motion=Free
  twistLimitDegrees=45
  twistMotion=Free
 }
 {
  bone1Name=dyn_skirtSIde_05
  bone2Name=dyn_skirtFrontSIde_05
  swing1LimitDegrees=45
  swing1Motion=Free
  swing2LimitDegrees=45
  swing2Motion=Free
  twistLimitDegrees=45
  twistMotion=Free
 }
 {
  bone1Name=dyn_skirtBackSIde_04
  bone2Name=dyn_skirtSIde_04
  swing1LimitDegrees=45
  swing1Motion=Free
  swing2LimitDegrees=45
  swing2Motion=Free
  twistLimitDegrees=45
  twistMotion=Free
 }
 {
  bone1Name=dyn_skirtFront_01_Base
  bone2Name=thigh_twist_01_l
  swing1LimitDegrees=45
  swing1Motion=Free
  swing2LimitDegrees=45
  swing2Motion=Free
  twistLimitDegrees=45
  twistMotion=Free
 }
 {
  bone1Name=dyn_skirtFrontSIde_01_Base
  bone2Name=thigh_twist_01_l
  swing1LimitDegrees=45
  swing1Motion=Free
  swing2LimitDegrees=45
  swing2Motion=Free
  twistLimitDegrees=45
  twistMotion=Free
 }
 {
  bone1Name=dyn_skirtSIde_01_Base
  bone2Name=thigh_twist_01_l
  swing1LimitDegrees=45
  swing1Motion=Free
  swing2LimitDegrees=45
  swing2Motion=Free
  twistLimitDegrees=45
  twistMotion=Free
 }
 {
  bone1Name=dyn_skirtFrontSIde_03
  bone2Name=dyn_skirtFront_03
  swing1LimitDegrees=45
  swing1Motion=Free
  swing2LimitDegrees=45
  swing2Motion=Free
  twistLimitDegrees=45
  twistMotion=Free
 }
 {
  bone1Name=foot_l
  bone2Name=calf_l
  swing1LimitDegrees=45
  swing1Motion=Free
  swing2LimitDegrees=45
  swing2Motion=Free
  twistLimitDegrees=45
  twistMotion=Free
 }
 {
  bone1Name=dyn_skirtSIde_02
  bone2Name=dyn_skirtBackSIde_02
  swing1LimitDegrees=45
  swing1Motion=Free
  swing2LimitDegrees=45
  swing2Motion=Free
  twistLimitDegrees=45
  twistMotion=Free
 }
 {
  bone1Name=dyn_skirtBackSIde_03
  bone2Name=dyn_skirtSIde_03
  swing1LimitDegrees=45
  swing1Motion=Free
  swing2LimitDegrees=45
  swing2Motion=Free
  twistLimitDegrees=45
  twistMotion=Free
 }
 {
  bone1Name=dyn_skirtSIde_04
  bone2Name=dyn_skirtFrontSIde_04
  swing1LimitDegrees=45
  swing1Motion=Free
  swing2LimitDegrees=45
  swing2Motion=Free
  twistLimitDegrees=45
  twistMotion=Free
 }
 {
  bone1Name=calf_twist_02_l
  bone2Name=calf_l
  swing1LimitDegrees=45
  swing1Motion=Free
  swing2LimitDegrees=45
  swing2Motion=Free
  twistLimitDegrees=45
  twistMotion=Free
 }
 {
  bone1Name=dyn_skirtFrontSIde_02
  bone2Name=dyn_skirtFront_02
  swing1LimitDegrees=45
  swing1Motion=Free
  swing2LimitDegrees=45
  swing2Motion=Free
  twistLimitDegrees=45
  twistMotion=Free
 }
 {
  bone1Name=dyn_skirtSIde_03
  bone2Name=dyn_skirtFrontSIde_03
  swing1LimitDegrees=45
  swing1Motion=Free
  swing2LimitDegrees=45
  swing2Motion=Free
  twistLimitDegrees=45
  twistMotion=Free
 }
 {
  bone1Name=dyn_skirtSIde_02
  bone2Name=dyn_skirtFrontSIde_02
  swing1LimitDegrees=45
  swing1Motion=Free
  swing2LimitDegrees=45
  swing2Motion=Free
  twistLimitDegrees=45
  twistMotion=Free
 }
 {
  bone1Name=dyn_skirtSIde_05
  bone2Name=dyn_skirtBackSIde_05
  swing1LimitDegrees=45
  swing1Motion=Free
  swing2LimitDegrees=45
  swing2Motion=Free
  twistLimitDegrees=45
  twistMotion=Free
 }
 {
  bone1Name=dyn_skirtFrontSIde_05
  bone2Name=dyn_skirtFront_05
  swing1LimitDegrees=45
  swing1Motion=Free
  swing2LimitDegrees=45
  swing2Motion=Free
  twistLimitDegrees=45
  twistMotion=Free
 }
 {
  bone1Name=calf_r
  bone2Name=root
  swing1LimitDegrees=45
  swing1Motion=Free
  swing2LimitDegrees=45
  swing2Motion=Free
  twistLimitDegrees=45
  twistMotion=Free
 }
]
