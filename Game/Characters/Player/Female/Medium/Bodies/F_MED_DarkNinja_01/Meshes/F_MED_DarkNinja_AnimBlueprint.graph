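# Post-process-style Anim Blueprint for the female Dark Ninja body, adding secondary motion and corrective deformation to straps, skirt and collar.
# It copies the attached parent's pose, then applies AnimDynamics straps, crouch/speed-driven skirt offsets and a Control Rig, cached as PreRigid.
# Additive bone transforms correct the collar for head, arm, spine and skydiving poses.
# Bools select between a constraint-driven no-rigids pose and rigid-body variants, including a front-end one.
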AnimBlueprint /Game/Characters/Player/Female/Medium/Bodies/F_MED_DarkNinja_01/Meshes/F_MED_DarkNinja_AnimBlueprint
target_skeleton=&/F_MED_DarkNinja_Skeleton
graphs [
 AnimGraph {nodes [
   n1 = "Transform (Modify) Bone - Bone: dyn_SkirtThigh_01_MSTR"!(ComponentPose, bAlphaBoolEnabled:bool=True, Translation="0,10,0", Rotation="0,-40,0", Scale="1,1,1") -> (Pose) {
    alphaInputType=Bool
    rotationMode=BMM_Additive
    rotationSpace=BCS_BoneSpace
    scaleSpace=BCS_BoneSpace
    translationMode=BMM_Additive
    translationSpace=BCS_BoneSpace
    alphaBoolBlend {blendInTime=0.1, blendOption=ExpInOut, blendOutTime=0.1}
    boneToModify {boneName=dyn_SkirtThigh_01_MSTR}
    rotation {yaw=-40}
    translation {y=10}
   }
   "Transform (Modify) Bone - Bone: dyn_SkirtPelvis_01_MSTR"!(ComponentPose=@n1, bAlphaBoolEnabled:bool=True, Translation="0,0,0", Rotation="0,-20,-30", Scale="1,1,1") -> (Pose) {
    alphaInputType=Bool
    rotationMode=BMM_Additive
    rotationSpace=BCS_BoneSpace
    scaleSpace=BCS_BoneSpace
    translationMode=BMM_Additive
    alphaBoolBlend {blendInTime=0.1, blendOption=ExpInOut, blendOutTime=0.1}
    boneToModify {boneName=dyn_SkirtPelvis_01_MSTR}
    rotation {roll=-30, yaw=-20}
   }
   n2 = Use cached pose 'rigids_cache'()
   n3 = Use cached pose 'rigids_FRONTEND_cache'()
   n4 = Get bIsInFrontEnd()
   Output Pose(Result=Blend Poses by bool(
     BlendPose_0=Use cached pose 'noRigids_cache'()
     BlendPose_1=Blend Poses by bool(BlendPose_0=@n3, BlendPose_1=@n2, bActiveValue=@n4) {blendType=Linear}
     bActiveValue=Get blendPoseByBoolValue() -> (blendPoseByBoolValue:bool)
    ) {blendType=Linear}) {layerGroup=None}
   n5 = Anim Dynamics - Bone: dyn_strap_r_Forearm_01(ComponentPose=Local To Component(LocalPose=Copy Pose From Mesh() {bUseAttachedParent=true}), Alpha=0.5, ExternalForce="0,0,0") {
    alpha=0.9
    angularSpringConstant=1500
    bAngularSpring=true
    bGravityOverrideInSimSpace=true
    numSolverIterationsPostUpdate=4
    numSolverIterationsPreUpdate=16
    simulationSpace=Actor
    alphaBoolBlend {blendOption=Linear}
    boundBone {boneName=dyn_strap_r_Forearm_01}
    physicsBodyDefinitions [{
      sphereCollisionRadius=0
      boundBone {boneName=dyn_strap_r_Forearm_01}
      boxExtents=<15 5 5>
      constraintSetup {angularTargetAxis=AxisZ, twistAxis=AxisZ, angularLimitsMax=<30 30 30>, angularLimitsMin=<-30 -30 -30>, angularTarget=<0 0 1>}
      localJointOffset=<0 0 0>
     }]
   }
   n6 = Anim Dynamics - Bone: dyn_strap_r_Forearm_02(ComponentPose=@n5, Alpha=0.25, ExternalForce="0,0,0") {
    alpha=0.9
    angularSpringConstant=750
    bAngularSpring=true
    bGravityOverrideInSimSpace=true
    numSolverIterationsPostUpdate=4
    numSolverIterationsPreUpdate=16
    simulationSpace=Actor
    alphaBoolBlend {blendOption=Linear}
    boundBone {boneName=dyn_strap_r_Forearm_02}
    physicsBodyDefinitions [{
      sphereCollisionRadius=0
      boundBone {boneName=dyn_strap_r_Forearm_02}
      boxExtents=<15 5 5>
      constraintSetup {angularTargetAxis=AxisZ, twistAxis=AxisZ, angularLimitsMax=<30 30 30>, angularLimitsMin=<-30 -30 -30>, angularTarget=<0 0 1>}
      localJointOffset=<0 0 0>
     }]
   }
   n7 = Anim Dynamics - Bone: dyn_strap_l_Forearm_01(ComponentPose=@n6, Alpha=0.5, ExternalForce="0,0,0") {
    alpha=0.9
    angularSpringConstant=1500
    bAngularSpring=true
    bGravityOverrideInSimSpace=true
    numSolverIterationsPostUpdate=4
    numSolverIterationsPreUpdate=16
    simulationSpace=Actor
    alphaBoolBlend {blendOption=Linear}
    boundBone {boneName=dyn_strap_l_Forearm_01}
    physicsBodyDefinitions [{
      sphereCollisionRadius=0
      boundBone {boneName=dyn_strap_l_Forearm_01}
      boxExtents=<15 5 5>
      constraintSetup {angularTargetAxis=AxisZ, twistAxis=AxisZ, angularLimitsMax=<30 30 30>, angularLimitsMin=<-30 -30 -30>, angularTarget=<0 0 1>}
      localJointOffset=<0 0 0>
     }]
   }
   n8 = Anim Dynamics - Bone: dyn_strap_l_Forearm_02(ComponentPose=@n7, Alpha=0.25, ExternalForce="0,0,0") {
    alpha=0.9
    angularSpringConstant=750
    bAngularSpring=true
    bGravityOverrideInSimSpace=true
    numSolverIterationsPostUpdate=4
    numSolverIterationsPreUpdate=16
    simulationSpace=Actor
    alphaBoolBlend {blendOption=Linear}
    boundBone {boneName=dyn_strap_l_Forearm_02}
    physicsBodyDefinitions [{
      sphereCollisionRadius=0
      boundBone {boneName=dyn_strap_l_Forearm_02}
      boxExtents=<15 5 5>
      constraintSetup {angularTargetAxis=AxisZ, twistAxis=AxisZ, angularLimitsMax=<30 30 30>, angularLimitsMin=<-30 -30 -30>, angularTarget=<0 0 1>}
      localJointOffset=<0 0 0>
     }]
   }
   n9 = Anim Dynamics - Bone: dyn_strap_r_Calf_01(ComponentPose=@n8, Alpha=0.15, ExternalForce="0,0,0") {
    alpha=0.9
    angularSpringConstant=1500
    bAngularSpring=true
    bGravityOverrideInSimSpace=true
    numSolverIterationsPostUpdate=4
    numSolverIterationsPreUpdate=16
    simulationSpace=Actor
    alphaBoolBlend {blendOption=Linear}
    boundBone {boneName=dyn_strap_r_Calf_01}
    physicsBodyDefinitions [{
      sphereCollisionRadius=0
      boundBone {boneName=dyn_strap_r_Calf_01}
      boxExtents=<15 5 5>
      constraintSetup {angularTargetAxis=AxisZ, twistAxis=AxisZ, angularLimitsMax=<30 30 30>, angularLimitsMin=<-30 -30 -30>, angularTarget=<0 0 1>}
      localJointOffset=<0 0 0>
     }]
   }
   n10 = Anim Dynamics - Bone: dyn_strap_r_Calf_02(ComponentPose=@n9, Alpha=0.3, ExternalForce="0,0,0") {
    alpha=0.9
    angularSpringConstant=1000
    bAngularSpring=true
    bGravityOverrideInSimSpace=true
    numSolverIterationsPostUpdate=4
    numSolverIterationsPreUpdate=16
    simulationSpace=Actor
    alphaBoolBlend {blendOption=Linear}
    boundBone {boneName=dyn_strap_r_Calf_02}
    physicsBodyDefinitions [{
      sphereCollisionRadius=0
      boundBone {boneName=dyn_strap_r_Calf_02}
      boxExtents=<15 5 5>
      constraintSetup {angularTargetAxis=AxisZ, twistAxis=AxisZ, angularLimitsMax=<30 30 30>, angularLimitsMin=<-30 -30 -30>, angularTarget=<0 0 1>}
      localJointOffset=<0 0 0>
     }]
   }
   n11 = Anim Dynamics - Bone: dyn_strap_r_Calf_03(ComponentPose=@n10, Alpha=0.5, ExternalForce="0,0,0") {
    alpha=0.9
    angularSpringConstant=500
    bAngularSpring=true
    bGravityOverrideInSimSpace=true
    numSolverIterationsPostUpdate=4
    numSolverIterationsPreUpdate=16
    simulationSpace=Actor
    alphaBoolBlend {blendOption=Linear}
    boundBone {boneName=dyn_strap_r_Calf_03}
    physicsBodyDefinitions [{
      sphereCollisionRadius=0
      boundBone {boneName=dyn_strap_r_Calf_03}
      boxExtents=<15 5 5>
      constraintSetup {angularTargetAxis=AxisZ, twistAxis=AxisZ, angularLimitsMax=<30 30 30>, angularLimitsMin=<-30 -30 -30>, angularTarget=<0 0 1>}
      localJointOffset=<0 0 0>
     }]
   }
   n12 = "Constraint - thigh_half_l (P)"(ComponentPose=@n11, Alpha=0.5, ConstraintWeights_0=1) {
    alphaBoolBlend {blendOption=Linear}
    boneToModify {boneName=thigh_half_l}
    constraintSetup [{targetBone {boneName=thigh_l}}]
    constraintWeights [1]
   }
   n13 = Get crouchAlpha() -> (crouchAlpha:double)
   n14 = "Transform (Modify) Bone - Bone: dyn_skirtSIde_01_Base"(ComponentPose=@n12, Alpha=@n13, Translation="0.000,-1.5,0.000", Rotation="-40,0.000,0.000", Scale="1,1,1") {
    rotationMode=BMM_Additive
    rotationSpace=BCS_BoneSpace
    translationMode=BMM_Additive
    translationSpace=BCS_BoneSpace
    alphaBoolBlend {blendOption=Linear}
    alphaScaleBiasClamp {bInterpResult=true, interpSpeedDecreasing=8, interpSpeedIncreasing=8}
    boneToModify {boneName=dyn_skirtSIde_01_Base}
    rotation {pitch=-40}
    translation {y=-1.5}
   }
   n15 = "Transform (Modify) Bone - Bone: dyn_skirtFrontSIde_01_Base"(ComponentPose=@n14, Alpha=@n13, Translation="0.000,-1.25,0.000", Rotation="-15,0.000,0.000", Scale="1,1,1") {
    rotationMode=BMM_Additive
    rotationSpace=BCS_BoneSpace
    translationMode=BMM_Additive
    translationSpace=BCS_BoneSpace
    alphaBoolBlend {blendOption=Linear}
    alphaScaleBiasClamp {bInterpResult=true, interpSpeedDecreasing=8, interpSpeedIncreasing=8}
    boneToModify {boneName=dyn_skirtFrontSIde_01_Base}
    rotation {pitch=-15}
    translation {y=-1.25}
   }
   n16 = "Transform (Modify) Bone - Bone: dyn_skirtFront_01_Base"(ComponentPose=@n15, Alpha=@n13, Translation="0.000,-2,-4", Rotation="-15.000,20.000,-32", Scale="1,1,1") {
    rotationMode=BMM_Additive
    rotationSpace=BCS_BoneSpace
    translationMode=BMM_Additive
    translationSpace=BCS_BoneSpace
    alphaBoolBlend {blendOption=Linear}
    alphaScaleBiasClamp {bInterpResult=true, interpSpeedDecreasing=8, interpSpeedIncreasing=8}
    boneToModify {boneName=dyn_skirtFront_01_Base}
    rotation=<rot -15 20 -32>
    translation {y=-2, z=-4}
   }
   n17 = "Constraint - dyn_skirtBackSIde_01_Base (PP)"!(ComponentPose=@n16, Alpha=1, ConstraintWeights_0=0.3, ConstraintWeights_1:float=0.7) -> (Pose) {
    alphaBoolBlend {blendOption=Linear}
    boneToModify {boneName=dyn_skirtBackSIde_01_Base}
    constraintSetup [{targetBone {boneName=thigh_twist_01_l}}, {targetBone {boneName=pelvis}}]
    constraintWeights [1,1]
   }
   n18 = Get speedAlpha() -> (speedAlpha:double)
   n19 = "Transform (Modify) Bone - Bone: dyn_skirtBack_01_Base"(ComponentPose=@n17, Alpha=@n18, Translation="0,0,0", Rotation="5,-5,0", Scale="1,1,1") {
    rotationMode=BMM_Additive
    rotationSpace=BCS_BoneSpace
    alphaBoolBlend {blendOption=Linear}
    alphaScaleBiasClamp {bInterpResult=true, interpSpeedDecreasing=3, interpSpeedIncreasing=3}
    boneToModify {boneName=dyn_skirtBack_01_Base}
    rotation {pitch=5, yaw=-5}
   }
   n20 = "Transform (Modify) Bone - Bone: dyn_skirtBackSIde_01_Base"(ComponentPose=@n19, Alpha=@n18, Translation="0,0,0", Rotation="15,-15,0", Scale="1,1,1") {
    rotationMode=BMM_Additive
    rotationSpace=BCS_BoneSpace
    alphaBoolBlend {blendOption=Linear}
    alphaScaleBiasClamp {bInterpResult=true, interpSpeedDecreasing=3, interpSpeedIncreasing=3}
    boneToModify {boneName=dyn_skirtBackSIde_01_Base}
    rotation {pitch=15, yaw=-15}
   }
   n21 = Control Rig(Source=Component To Local(ComponentPose=@n20)) {
    alphaBoolBlend {blendOption=Linear}
    controlRigAssetReference {blueprintRigClass=&/F_MED_DarkNinja_CtrlRig.F_MED_DarkNinja_CtrlRig_C}
   }
   Save cached pose 'PreRigid'(Pose=@n21) {cacheName=PreRigid}
   n22 = "Constraint - dyn_skirtFront_04 (P)"(ComponentPose=Local To Component(LocalPose=Use cached pose 'PreRigid'()), Alpha=1, ConstraintWeights_0=1) {
    alphaBoolBlend {blendOption=Linear}
    boneToModify {boneName=dyn_skirtFront_04}
    constraintSetup [{targetBone {boneName=calf_l}}]
    constraintWeights [1]
   }
   n23 = "Constraint - dyn_skirtFrontSIde_04 (P)"(ComponentPose=@n22, Alpha=1, ConstraintWeights_0=1) {
    alphaBoolBlend {blendOption=Linear}
    boneToModify {boneName=dyn_skirtFrontSIde_04}
    constraintSetup [{targetBone {boneName=calf_l}}]
    constraintWeights [1]
   }
   n24 = "Constraint - dyn_skirtSIde_04 (P)"(ComponentPose=@n23, Alpha=1, ConstraintWeights_0=1) {
    alphaBoolBlend {blendOption=Linear}
    boneToModify {boneName=dyn_skirtSIde_04}
    constraintSetup [{targetBone {boneName=calf_l}}]
    constraintWeights [1]
   }
   n25 = "Constraint - dyn_skirtBackSIde_02 (P)"(ComponentPose=@n24, Alpha=1, ConstraintWeights_0=1) {
    alphaBoolBlend {blendOption=Linear}
    boneToModify {boneName=dyn_skirtBackSIde_02}
    constraintSetup [{targetBone {boneName=thigh_l}}]
    constraintWeights [1]
   }
   n26 = "Constraint - dyn_skirtBack_02 (P)"(ComponentPose=@n25, Alpha=1, ConstraintWeights_0=1) {
    alphaBoolBlend {blendOption=Linear}
    boneToModify {boneName=dyn_skirtBack_02}
    constraintSetup [{targetBone {boneName=thigh_l}}]
    constraintWeights [1]
   }
   n27 = "Constraint - dyn_skirtBack_03 (P)"(ComponentPose=@n26, Alpha=1, ConstraintWeights_0=1) {
    alphaBoolBlend {blendOption=Linear}
    boneToModify {boneName=dyn_skirtBack_03}
    constraintSetup [{targetBone {boneName=calf_l}}]
    constraintWeights [1]
   }
   n28 = "Constraint - dyn_skirtBackSIde_03 (P)"(ComponentPose=@n27, Alpha=1, ConstraintWeights_0=1) {
    alphaBoolBlend {blendOption=Linear}
    boneToModify {boneName=dyn_skirtBackSIde_03}
    constraintSetup [{targetBone {boneName=calf_l}}]
    constraintWeights [1]
   }
   n29 = "Transform (Modify) Bone - Bone: dyn_skirtBack_04"(ComponentPose=@n28, Alpha=1, Translation="0,-4,0", Rotation="0,0,0", Scale="1,1,1") {
    translationMode=BMM_Additive
    translationSpace=BCS_BoneSpace
    alphaBoolBlend {blendOption=Linear}
    boneToModify {boneName=dyn_skirtBack_04}
    translation {y=-4}
   }
   Save cached pose 'noRigids_cache'(Pose=Component To Local(ComponentPose=@n29)) {cacheName=noRigids_cache}
   n30 = RigidBody!(
    ComponentPose=Local To Component(LocalPose=Use cached pose 'Post-skirt setup'())
    Alpha=Get notInVehicle() -> (notInVehicle:double)
    OverrideWorldGravity:Vector=Get GravityOverride()
    ExternalForce=Get skydiving vector force() -> (skydiving vector force:Vector)
   ) -> (Pose) {
    bClampLinearTranslationLimitToRefPose=true
    bOverrideWorldGravity=true
    overridePhysicsAsset=&/F_MED_DarkNinja_FrontEnd_Physics
    simulationSpace=BaseBoneSpace
    alphaBoolBlend {blendOption=Linear}
    baseBoneRef {boneName=thigh_l}
    componentAppliedLinearAccClamp=<100 100 100>
    componentLinearAccScale=<0.1 0.1 0.1>
    componentLinearVelScale=<0.1 0.1 0.1>
    overrideWorldGravity {z=-980}
    simSpaceSettings {maxAngularAcceleration=1000, maxAngularVelocity=3, maxLinearAcceleration=250, maxLinearVelocity=500, velocityScaleZ=0, worldAlpha=1}
   }
   n31 = Get CustomIdleIsPlaying() -> (CustomIdleIsPlaying:float)
   n32 = "Transform (Modify) Bone - Bone: dyn_skirtFront_01_Base"(ComponentPose=@n30, Alpha=@n31, Translation="0.5,0.5,0", Rotation="0,0,0", Scale="1,1,1") {
    rotationMode=BMM_Additive
    scaleMode=BMM_Additive
    translationMode=BMM_Additive
    alphaBoolBlend {blendOption=Linear}
    boneToModify {boneName=dyn_skirtFront_01_Base}
    translation=<0.5 0.5>
   }
   n33 = "Transform (Modify) Bone - Bone: dyn_skirtFrontSIde_01_Base"(ComponentPose=@n32, Alpha=@n31, Translation="-3.37,0.501,0", Rotation="6,0,0", Scale="1,1,1") {
    rotationMode=BMM_Additive
    scaleMode=BMM_Additive
    translationMode=BMM_Additive
    alphaBoolBlend {blendOption=Linear}
    boneToModify {boneName=dyn_skirtFrontSIde_01_Base}
    rotation {pitch=6}
    translation=<-3.37 0.501>
   }
   n34 = "Transform (Modify) Bone - Bone: dyn_skirtSIde_01_Base"(ComponentPose=@n33, Alpha=@n31, Translation="-1,0,0", Rotation="8,0,0", Scale="1,1,1") {
    rotationMode=BMM_Additive
    scaleMode=BMM_Additive
    translationMode=BMM_Additive
    alphaBoolBlend {blendOption=Linear}
    boneToModify {boneName=dyn_skirtSIde_01_Base}
    rotation {pitch=8}
    translation {x=-1}
   }
   n35 = "Transform (Modify) Bone - Bone: dyn_skirtSIde_03"(ComponentPose=@n34, Alpha=@n31, Translation="0,0,0", Rotation="-6,0,0", Scale="1,1,1") {
    rotationMode=BMM_Additive
    scaleMode=BMM_Additive
    translationMode=BMM_Additive
    alphaBoolBlend {blendOption=Linear}
    boneToModify {boneName=dyn_skirtSIde_03}
    rotation {pitch=-6}
   }
   n36 = "Transform (Modify) Bone - Bone: dyn_SkirtPelvis_01_MSTR"(ComponentPose=@n35, bAlphaBoolEnabled=True, Translation="0,0,0", Rotation="14.199,-48.453,-31.029", Scale="1,1,1") {
    alphaInputType=Bool
    rotationMode=BMM_Additive
    rotationSpace=BCS_BoneSpace
    translationMode=BMM_Additive
    translationSpace=BCS_BoneSpace
    alphaBoolBlend {blendInTime=0.3, blendOption=ExpInOut, blendOutTime=0.3}
    alphaScaleBiasClamp {bInterpResult=true}
    boneToModify {boneName=dyn_SkirtPelvis_01_MSTR}
    rotation=<rot 14.199 -48.4529 -31.029>
   }
   n37 = "Transform (Modify) Bone - Bone: dyn_SkirtThigh_01_MSTR"(ComponentPose=@n36, bAlphaBoolEnabled=True, Translation="0,0,0", Rotation="42.289,-19.905,-18.110", Scale="1,1,1") {
    alphaInputType=Bool
    rotationMode=BMM_Additive
    rotationSpace=BCS_BoneSpace
    translationMode=BMM_Additive
    translationSpace=BCS_BoneSpace
    alphaBoolBlend {blendInTime=0.3, blendOption=ExpInOut, blendOutTime=0.3}
    alphaScaleBiasClamp {bInterpResult=true}
    boneToModify {boneName=dyn_SkirtThigh_01_MSTR}
    rotation=<rot 42.289 -19.905 -18.11>
   }
   n38 = "Transform (Modify) Bone - Bone: dyn_skirtBack_03"!(ComponentPose=@n37, bAlphaBoolEnabled:bool=True, Translation="0,0,0", Rotation="-14.367,69.420,3.443", Scale="1,1,1") -> (Pose) {
    alphaInputType=Bool
    rotationMode=BMM_Additive
    rotationSpace=BCS_BoneSpace
    translationMode=BMM_Additive
    translationSpace=BCS_BoneSpace
    alphaBoolBlend {blendInTime=0.3, blendOption=ExpInOut, blendOutTime=0.3}
    alphaScaleBiasClamp {bInterpResult=true}
    boneToModify {boneName=dyn_skirtBack_03}
    rotation=<rot -14.367 69.42 3.4429>
   }
   Save cached pose 'rigids_FRONTEND_cache'(Pose=Component To Local(ComponentPose=@n38)) {cacheName=rigids_FRONTEND_cache}
   n39 = Get Melee Head turn to the left Alpha() -> (Melee Head turn to the left Alpha:double)
   n40 = Get Melee Head turn to the right Alpha() -> (Melee Head turn to the right Alpha:double)
   n41 = Get Skydiving Movement Forward or Backward Alpha() -> (Skydiving Movement Forward or Backward Alpha:double)
   n42 = Get Head turn to the right Alpha() -> (Head turn to the right Alpha:double)
   n43 = Get L Arm Raise Alpha() -> (L Arm Raise Alpha:float)
   n44 = Get R Arm Raise Alpha() -> (R Arm Raise Alpha:float)
   n45 = "Transform (Modify) Bone - Bone: deform_collar_front_r_01"(
    ComponentPose=Local To Component(LocalPose=Use cached pose 'PreRigid'())
    Alpha=@n44
    Translation="-3.014,-0.092,-3.188"
    Rotation="-49.860,-2.345,1.441"
    Scale="1,1,1"
   ) {
    rotationMode=BMM_Additive
    rotationSpace=BCS_BoneSpace
    translationMode=BMM_Additive
    translationSpace=BCS_BoneSpace
    alphaBoolBlend {blendOption=Linear}
    boneToModify {boneName=deform_collar_front_r_01}
    rotation=<rot -49.86 -2.3451 1.4412>
    translation=<-3.014 -0.092 -3.188>
   }
   n46 = "Transform (Modify) Bone - Bone: deform_collar_back_r_01"(ComponentPose=@n45, Alpha=@n44, Translation="0.287,3.985,4.236", Rotation="0.000,0.046,-0.010", Scale="1,1,1") {
    rotationMode=BMM_Additive
    rotationSpace=BCS_BoneSpace
    translationMode=BMM_Additive
    translationSpace=BCS_BoneSpace
    alphaBoolBlend {blendOption=Linear}
    boneToModify {boneName=deform_collar_back_r_01}
    rotation=<rot 0.0001 0.046 -0.0099>
    translation=<0.287 3.985 4.236>
   }
   n47 = Get Spine rotation to the Left while Targeting Alpha() -> (Spine rotation to the Left while Targeting Alpha:double)
   n48 = "Transform (Modify) Bone - Bone: deform_collar_front_r_01"(ComponentPose=@n46, Alpha=@n47, Translation="4.857,0.156,0", Rotation="4.938,0.779,0.143", Scale="1,1,1") {
    rotationMode=BMM_Additive
    rotationSpace=BCS_BoneSpace
    translationMode=BMM_Additive
    translationSpace=BCS_BoneSpace
    alphaBoolBlend {blendOption=Linear}
    boneToModify {boneName=deform_collar_front_r_01}
    rotation=<rot 4.938 0.779 0.143>
    translation=<4.857 0.156>
   }
   n49 = "Transform (Modify) Bone - Bone: deform_collar_back_r_01"(ComponentPose=@n48, Alpha=@n47, Translation="-0.777,3.400,0.023", Rotation="0,0,0", Scale="1,1,1") {
    rotationMode=BMM_Additive
    rotationSpace=BCS_BoneSpace
    translationMode=BMM_Additive
    translationSpace=BCS_BoneSpace
    alphaBoolBlend {blendOption=Linear}
    boneToModify {boneName=deform_collar_back_r_01}
    translation=<-0.777 3.4 0.023>
   }
   n50 = "Transform (Modify) Bone - Bone: deform_collar_back_l_01"(ComponentPose=@n49, Alpha=@n47, Translation="0.888,0.696,-2.269", Rotation="0,0,0", Scale="1,1,1") {
    rotationMode=BMM_Additive
    rotationSpace=BCS_BoneSpace
    translationMode=BMM_Additive
    translationSpace=BCS_BoneSpace
    alphaBoolBlend {blendOption=Linear}
    boneToModify {boneName=deform_collar_back_l_01}
    translation=<0.888 0.696 -2.269>
   }
   n51 = "Transform (Modify) Bone - Bone: deform_collar_front_l_01"(ComponentPose=@n50, Alpha=@n47, Translation="0,0,0", Rotation="0,0,0", Scale="1,1,1") {
    rotationMode=BMM_Additive
    rotationSpace=BCS_BoneSpace
    translationMode=BMM_Additive
    translationSpace=BCS_BoneSpace
    alphaBoolBlend {blendOption=Linear}
    boneToModify {boneName=deform_collar_front_l_01}
   }
   n52 = Get Spine rotation to the Right while Targeting Alpha() -> (Spine rotation to the Right while Targeting Alpha:double)
   n53 = "Transform (Modify) Bone - Bone: deform_collar_front_r_01"(ComponentPose=@n51, Alpha=@n52, Translation="0,0,0", Rotation="0,0,0", Scale="1,1,1") {
    rotationMode=BMM_Additive
    rotationSpace=BCS_BoneSpace
    translationMode=BMM_Additive
    translationSpace=BCS_BoneSpace
    alphaBoolBlend {blendOption=Linear}
    boneToModify {boneName=deform_collar_front_r_01}
   }
   n54 = "Transform (Modify) Bone - Bone: deform_collar_back_r_01"(ComponentPose=@n53, Alpha=@n52, Translation="0.703,0.619,2.999", Rotation="0,0,0", Scale="1,1,1") {
    rotationMode=BMM_Additive
    rotationSpace=BCS_BoneSpace
    translationMode=BMM_Additive
    translationSpace=BCS_BoneSpace
    alphaBoolBlend {blendOption=Linear}
    boneToModify {boneName=deform_collar_back_r_01}
    translation=<0.703 0.619 2.999>
   }
   n55 = "Transform (Modify) Bone - Bone: deform_collar_back_l_01"(ComponentPose=@n54, Alpha=@n52, Translation="1.789,2.395,-2.182", Rotation="-11.942,13.184,-15.058", Scale="1,1,1") {
    rotationMode=BMM_Additive
    rotationSpace=BCS_BoneSpace
    translationMode=BMM_Additive
    translationSpace=BCS_BoneSpace
    alphaBoolBlend {blendOption=Linear}
    boneToModify {boneName=deform_collar_back_l_01}
    rotation=<rot -11.942 13.184 -15.058>
    translation=<1.789 2.395 -2.182>
   }
   n56 = "Transform (Modify) Bone - Bone: deform_collar_front_l_01"(ComponentPose=@n55, Alpha=@n52, Translation="-0.180,-0.003,0.735", Rotation="0,0,0", Scale="1,1,1") {
    rotationMode=BMM_Additive
    rotationSpace=BCS_BoneSpace
    translationMode=BMM_Additive
    translationSpace=BCS_BoneSpace
    alphaBoolBlend {blendOption=Linear}
    boneToModify {boneName=deform_collar_front_l_01}
    translation=<-0.18 -0.003 0.735>
   }
   n57 = "Transform (Modify) Bone - Bone: deform_collar_front_l_01"(ComponentPose=@n56, Alpha=@n43, Translation="-1.135,0.012,-2.293", Rotation="-20.000,0.004,0.004", Scale="1,1,1") {
    rotationMode=BMM_Additive
    rotationSpace=BCS_BoneSpace
    translationMode=BMM_Additive
    translationSpace=BCS_BoneSpace
    alphaBoolBlend {blendOption=Linear}
    boneToModify {boneName=deform_collar_front_l_01}
    rotation=<rot -20.0001 0.0041 0.004>
    translation=<-1.135 0.012 -2.293>
   }
   n58 = "Transform (Modify) Bone - Bone: deform_collar_back_l_01"(ComponentPose=@n57, Alpha=@n43, Translation="1.355,2.996,-2.453", Rotation="0.013,0.013,0.023", Scale="1,1,1") {
    rotationMode=BMM_Additive
    rotationSpace=BCS_BoneSpace
    translationMode=BMM_Additive
    translationSpace=BCS_BoneSpace
    alphaBoolBlend {blendOption=Linear}
    boneToModify {boneName=deform_collar_back_l_01}
    rotation=<rot 0.013 0.013 0.023>
    translation=<1.355 2.996 -2.453>
   }
   n59 = Get Head look Up Alpha() -> (Head look Up Alpha:double)
   n60 = "Transform (Modify) Bone - Bone: deform_collar_back_r_01"(ComponentPose=@n58, Alpha=@n59, Translation="2.898,-3.032,-0.522", Rotation="0,0,0", Scale="1,1,1") {
    rotationMode=BMM_Additive
    rotationSpace=BCS_BoneSpace
    translationMode=BMM_Additive
    translationSpace=BCS_BoneSpace
    alphaBoolBlend {blendOption=Linear}
    boneToModify {boneName=deform_collar_back_r_01}
    translation=<2.898 -3.032 -0.522>
   }
   n61 = "Transform (Modify) Bone - Bone: deform_collar_back_01"(ComponentPose=@n60, Alpha=@n59, Translation="-8.163,-4.433,-0.116", Rotation="-5.586,-23.685,14.184", Scale="1,1,1") {
    rotationMode=BMM_Additive
    rotationSpace=BCS_BoneSpace
    translationMode=BMM_Additive
    translationSpace=BCS_BoneSpace
    alphaBoolBlend {blendOption=Linear}
    boneToModify {boneName=deform_collar_back_01}
    rotation=<rot -5.586 -23.685 14.184>
    translation=<-8.163 -4.433 -0.116>
   }
   n62 = "Transform (Modify) Bone - Bone: deform_collar_back_l_01"(ComponentPose=@n61, Alpha=@n59, Translation="-1.084,3.121,5.762", Rotation="25.856,15.552,-10.589", Scale="1,1,1") {
    rotationMode=BMM_Additive
    rotationSpace=BCS_BoneSpace
    translationMode=BMM_Additive
    translationSpace=BCS_BoneSpace
    alphaBoolBlend {blendOption=Linear}
    boneToModify {boneName=deform_collar_back_l_01}
    rotation=<rot 25.856 15.5521 -10.589>
    translation=<-1.084 3.121 5.762>
   }
   n63 = Get Head Tilt To Right Alpha() -> (Head Tilt To Right Alpha:double)
   n64 = "Transform (Modify) Bone - Bone: deform_collar_back_r_01"(ComponentPose=@n62, Alpha=@n63, Translation="-2.474,3.172,0.264", Rotation="38.824,4.609,43.149", Scale="1,1,1") {
    rotationMode=BMM_Additive
    rotationSpace=BCS_BoneSpace
    translationMode=BMM_Additive
    translationSpace=BCS_BoneSpace
    alphaBoolBlend {blendOption=Linear}
    boneToModify {boneName=deform_collar_back_r_01}
    rotation=<rot 38.824 4.609 43.149>
    translation=<-2.474 3.172 0.264>
   }
   n65 = "Transform (Modify) Bone - Bone: deform_collar_back_01"(ComponentPose=@n64, Alpha=@n63, Translation="0,0,-0.637", Rotation="0.022,0.001,-15.000", Scale="1,1,1") {
    rotationMode=BMM_Additive
    rotationSpace=BCS_BoneSpace
    translationMode=BMM_Additive
    translationSpace=BCS_BoneSpace
    alphaBoolBlend {blendOption=Linear}
    boneToModify {boneName=deform_collar_back_01}
    rotation=<rot 0.022 0.0009 -15.0001>
    translation {z=-0.637}
   }
   n66 = "Transform (Modify) Bone - Bone: deform_collar_front_r_01"(ComponentPose=@n65, Alpha=@n63, Translation="0.563,-0.150,1.650", Rotation="-24.545,-5.752,10.943", Scale="1,1,1") {
    rotationMode=BMM_Additive
    rotationSpace=BCS_BoneSpace
    translationMode=BMM_Additive
    translationSpace=BCS_BoneSpace
    alphaBoolBlend {blendOption=Linear}
    boneToModify {boneName=deform_collar_front_r_01}
    rotation=<rot -24.545 -5.7521 10.9431>
    translation=<0.5631 -0.15 1.6502>
   }
   n67 = Get Head turn to the left Alpha() -> (Head turn to the left Alpha:double)
   n68 = "Transform (Modify) Bone - Bone: deform_collar_back_r_01"(ComponentPose=@n66, Alpha=@n67, Translation="4.406,0.234,2.622", Rotation="0,0,0", Scale="1,1,1") {
    rotationMode=BMM_Additive
    rotationSpace=BCS_BoneSpace
    translationMode=BMM_Additive
    translationSpace=BCS_BoneSpace
    alphaBoolBlend {blendOption=Linear}
    boneToModify {boneName=deform_collar_back_r_01}
    translation=<4.406 0.234 2.622>
   }
   n69 = "Transform (Modify) Bone - Bone: deform_collar_front_r_01"(ComponentPose=@n68, Alpha=@n67, Translation="-0.962,-0.011,-3.273", Rotation="5.000,-0.002,0.002", Scale="1,1,1") {
    rotationMode=BMM_Additive
    rotationSpace=BCS_BoneSpace
    translationMode=BMM_Additive
    translationSpace=BCS_BoneSpace
    alphaBoolBlend {blendOption=Linear}
    boneToModify {boneName=deform_collar_front_r_01}
    rotation=<rot 5 -0.002 0.002>
    translation=<-0.962 -0.011 -3.273>
   }
   n70 = "Transform (Modify) Bone - Bone: deform_collar_back_r_01"(ComponentPose=@n69, Alpha=@n42, Translation="-1.125,-0.000,3.257", Rotation="60,0,0", Scale="1,1,1") {
    rotationMode=BMM_Additive
    rotationSpace=BCS_BoneSpace
    translationMode=BMM_Additive
    translationSpace=BCS_BoneSpace
    alphaBoolBlend {blendOption=Linear}
    boneToModify {boneName=deform_collar_back_r_01}
    rotation {pitch=60}
    translation {x=-1.125, z=3.257}
   }
   n71 = "Transform (Modify) Bone - Bone: deform_collar_back_01"(ComponentPose=@n70, Alpha=@n42, Translation="0.000,-3.422,-0.782", Rotation="0.000,-0.000,-25.000", Scale="1,1,1") {
    rotationMode=BMM_Additive
    rotationSpace=BCS_BoneSpace
    translationMode=BMM_Additive
    translationSpace=BCS_BoneSpace
    alphaBoolBlend {blendOption=Linear}
    boneToModify {boneName=deform_collar_back_01}
    rotation=<rot 0 0 -25.0001>
    translation {y=-3.422, z=-0.782}
   }
   n72 = "Transform (Modify) Bone - Bone: deform_collar_back_l_01"!(ComponentPose=@n71, bAlphaBoolEnabled:bool=True, Translation="0.588,-0.016,0.786", Rotation="25.000,0.000,-0.000", Scale="1,1,1") -> (Pose) {
    alphaInputType=Bool
    rotationMode=BMM_Additive
    rotationSpace=BCS_BoneSpace
    translationMode=BMM_Additive
    translationSpace=BCS_BoneSpace
    alphaBoolBlend {blendInTime=0.3, blendOption=ExpInOut, blendOutTime=0.3}
    alphaScaleBiasClamp {bInterpResult=true}
    boneToModify {boneName=deform_collar_back_l_01}
    rotation=<rot 25 0 0>
    translation=<0.588 -0.016 0.786>
   }
   n73 = "Transform (Modify) Bone - Bone: deform_collar_back_01"!(ComponentPose=@n72, bAlphaBoolEnabled:bool=True, Translation="-0.000,-1.873,0.000", Rotation="0,0,0", Scale="1,1,1") -> (Pose) {
    alphaInputType=Bool
    rotationMode=BMM_Additive
    rotationSpace=BCS_BoneSpace
    translationMode=BMM_Additive
    translationSpace=BCS_BoneSpace
    alphaBoolBlend {blendInTime=0.3, blendOption=ExpInOut, blendOutTime=0.3}
    alphaScaleBiasClamp {bInterpResult=true}
    boneToModify {boneName=deform_collar_back_01}
    translation {y=-1.8726}
   }
   n74 = "Transform (Modify) Bone - Bone: deform_collar_back_r_01"!(ComponentPose=@n73, bAlphaBoolEnabled:bool=True, Translation="0.000,0.169,3.867", Rotation="0,0,0", Scale="1,1,1") -> (Pose) {
    alphaInputType=Bool
    rotationMode=BMM_Additive
    rotationSpace=BCS_BoneSpace
    translationMode=BMM_Additive
    translationSpace=BCS_BoneSpace
    alphaBoolBlend {blendInTime=0.3, blendOption=ExpInOut, blendOutTime=0.3}
    alphaScaleBiasClamp {bInterpResult=true}
    boneToModify {boneName=deform_collar_back_r_01}
    translation {y=0.169, z=3.867}
   }
   n75 = "Transform (Modify) Bone - Bone: deform_collar_front_r_01"!(ComponentPose=@n74, bAlphaBoolEnabled:bool=True, Translation="-0.000,-0.000,3.988", Rotation="0,0,0", Scale="1,1,1") -> (Pose) {
    alphaInputType=Bool
    rotationMode=BMM_Additive
    rotationSpace=BCS_BoneSpace
    translationMode=BMM_Additive
    translationSpace=BCS_BoneSpace
    alphaBoolBlend {blendInTime=0.3, blendOption=ExpInOut, blendOutTime=0.3}
    alphaScaleBiasClamp {bInterpResult=true}
    boneToModify {boneName=deform_collar_front_r_01}
    translation {z=3.9884}
   }
   n76 = "Transform (Modify) Bone - Bone: deform_collar_front_l_01"!(ComponentPose=@n75, bAlphaBoolEnabled:bool=True, Translation="-0.000,0.000,2.024", Rotation="0,0,0", Scale="1,1,1") -> (Pose) {
    alphaInputType=Bool
    rotationMode=BMM_Additive
    rotationSpace=BCS_BoneSpace
    translationMode=BMM_Additive
    translationSpace=BCS_BoneSpace
    alphaBoolBlend {blendInTime=0.3, blendOption=ExpInOut, blendOutTime=0.3}
    alphaScaleBiasClamp {bInterpResult=true}
    boneToModify {boneName=deform_collar_front_l_01}
    translation {z=2.024}
   }
   n77 = "Transform (Modify) Bone - Bone: deform_collar_back_r_01"!(
    ComponentPose=@n76
    Alpha:float=Get Skydiving R Arm Raise Alpha() -> (Skydiving R Arm Raise Alpha:float)
    Translation="0.000,-0.000,-10.653"
    Rotation="0,0,0"
    Scale="1,1,1"
   ) -> (Pose) {
    rotationMode=BMM_Additive
    rotationSpace=BCS_BoneSpace
    translationMode=BMM_Additive
    translationSpace=BCS_BoneSpace
    alphaBoolBlend {blendInTime=0.3, blendOption=ExpInOut, blendOutTime=0.3}
    boneToModify {boneName=deform_collar_back_r_01}
    translation {z=-10.653}
   }
   n78 = Get Skydiving Movement to Right Alpha() -> (Skydiving Movement to Right Alpha:double)
   n79 = "Transform (Modify) Bone - Bone: deform_collar_back_l_01"!(ComponentPose=@n77, Alpha:float=@n78, Translation="-0.000,0.000,-2.502", Rotation="0,0,0", Scale="1,1,1") -> (Pose) {
    rotationMode=BMM_Additive
    rotationSpace=BCS_BoneSpace
    translationMode=BMM_Additive
    translationSpace=BCS_BoneSpace
    alphaBoolBlend {blendInTime=0.3, blendOption=ExpInOut, blendOutTime=0.3}
    boneToModify {boneName=deform_collar_back_l_01}
    translation {z=-2.502}
   }
   n80 = "Transform (Modify) Bone - Bone: deform_collar_back_r_01"(ComponentPose=@n79, Alpha=@n78, Translation="-0.000,-0.229,-4.148", Rotation="0,0,0", Scale="1,1,1") {
    rotationMode=BMM_Additive
    rotationSpace=BCS_BoneSpace
    translationMode=BMM_Additive
    translationSpace=BCS_BoneSpace
    alphaBoolBlend {blendInTime=0.3, blendOption=ExpInOut, blendOutTime=0.3}
    boneToModify {boneName=deform_collar_back_r_01}
    translation {y=-0.229, z=-4.148}
   }
   n81 = Get Skydiving Movement to Left Alpha() -> (Skydiving Movement to Left Alpha:double)
   n82 = "Transform (Modify) Bone - Bone: deform_collar_back_l_01"(ComponentPose=@n80, Alpha=@n81, Translation="-0.000,-0.000,-6.098", Rotation="0,0,0", Scale="1,1,1") {
    rotationMode=BMM_Additive
    rotationSpace=BCS_BoneSpace
    translationMode=BMM_Additive
    translationSpace=BCS_BoneSpace
    alphaBoolBlend {blendInTime=0.3, blendOption=ExpInOut, blendOutTime=0.3}
    boneToModify {boneName=deform_collar_back_l_01}
    translation {z=-6.098}
   }
   n83 = "Transform (Modify) Bone - Bone: deform_collar_back_r_01"(ComponentPose=@n82, Alpha=@n81, Translation="-0.000,-0.000,-0.877", Rotation="0,0,0", Scale="1,1,1") {
    rotationMode=BMM_Additive
    rotationSpace=BCS_BoneSpace
    translationMode=BMM_Additive
    translationSpace=BCS_BoneSpace
    alphaBoolBlend {blendInTime=0.3, blendOption=ExpInOut, blendOutTime=0.3}
    boneToModify {boneName=deform_collar_back_r_01}
    translation {z=-0.877}
   }
   n84 = "Transform (Modify) Bone - Bone: deform_collar_back_l_01"(ComponentPose=@n83, Alpha=@n41, Translation="0.177,-0.000,-1.819", Rotation="0,0,0", Scale="1,1,1") {
    rotationMode=BMM_Additive
    rotationSpace=BCS_BoneSpace
    translationMode=BMM_Additive
    translationSpace=BCS_BoneSpace
    alphaBoolBlend {blendInTime=0.3, blendOption=ExpInOut, blendOutTime=0.3}
    boneToModify {boneName=deform_collar_back_l_01}
    translation {x=0.177, z=-1.819}
   }
   n85 = "Transform (Modify) Bone - Bone: deform_collar_back_01"!(ComponentPose=@n84, Alpha:float=@n41, Translation="0,0,0", Rotation="0,0,0", Scale="1,1,1") -> (Pose) {
    rotationMode=BMM_Additive
    rotationSpace=BCS_BoneSpace
    translationMode=BMM_Additive
    translationSpace=BCS_BoneSpace
    alphaBoolBlend {blendInTime=0.3, blendOption=ExpInOut, blendOutTime=0.3}
    boneToModify {boneName=deform_collar_back_01}
   }
   n86 = "Transform (Modify) Bone - Bone: deform_collar_back_r_01"(ComponentPose=@n85, Alpha=@n41, Translation="0.000,0.000,4.895", Rotation="0,0,0", Scale="1,1,1") {
    rotationMode=BMM_Additive
    rotationSpace=BCS_BoneSpace
    translationMode=BMM_Additive
    translationSpace=BCS_BoneSpace
    alphaBoolBlend {blendInTime=0.3, blendOption=ExpInOut, blendOutTime=0.3}
    boneToModify {boneName=deform_collar_back_r_01}
    translation {z=4.895}
   }
   n87 = "Transform (Modify) Bone - Bone: deform_collar_front_r_01"!(ComponentPose=@n86, Alpha:float=@n41, Translation="0.000,-0.000,-3.473", Rotation="0,0,0", Scale="1,1,1") -> (Pose) {
    rotationMode=BMM_Additive
    rotationSpace=BCS_BoneSpace
    translationMode=BMM_Additive
    translationSpace=BCS_BoneSpace
    alphaBoolBlend {blendInTime=0.3, blendOption=ExpInOut, blendOutTime=0.3}
    boneToModify {boneName=deform_collar_front_r_01}
    translation {z=-3.473}
   }
   n88 = "Transform (Modify) Bone - Bone: deform_collar_front_l_01"!(ComponentPose=@n87, Alpha:float=@n41, Translation="-0.000,0.000,-2.599", Rotation="0,0,0", Scale="1,1,1") -> (Pose) {
    rotationMode=BMM_Additive
    rotationSpace=BCS_BoneSpace
    translationMode=BMM_Additive
    translationSpace=BCS_BoneSpace
    alphaBoolBlend {blendInTime=0.3, blendOption=ExpInOut, blendOutTime=0.3}
    boneToModify {boneName=deform_collar_front_l_01}
    translation {z=-2.599}
   }
   n89 = "Transform (Modify) Bone - Bone: deform_collar_back_l_01"!(ComponentPose=@n88, bAlphaBoolEnabled:bool=True, Translation="-0.000,-0.000,1.511", Rotation="0,0,0", Scale="1,1,1") -> (Pose) {
    alphaInputType=Bool
    rotationMode=BMM_Additive
    rotationSpace=BCS_BoneSpace
    translationMode=BMM_Additive
    translationSpace=BCS_BoneSpace
    alphaBoolBlend {blendInTime=0.3, blendOption=ExpInOut, blendOutTime=0.3}
    alphaScaleBiasClamp {bInterpResult=true}
    boneToModify {boneName=deform_collar_back_l_01}
    translation {z=1.511}
   }
   n90 = "Transform (Modify) Bone - Bone: deform_collar_back_01"!(ComponentPose=@n89, bAlphaBoolEnabled:bool=True, Translation="-0.000,-2.601,-0.000", Rotation="0,0,0", Scale="1,1,1") -> (Pose) {
    alphaInputType=Bool
    rotationMode=BMM_Additive
    rotationSpace=BCS_BoneSpace
    translationMode=BMM_Additive
    translationSpace=BCS_BoneSpace
    alphaBoolBlend {blendInTime=0.3, blendOption=ExpInOut, blendOutTime=0.3}
    alphaScaleBiasClamp {bInterpResult=true}
    boneToModify {boneName=deform_collar_back_01}
    translation {y=-2.601}
   }
   n91 = "Transform (Modify) Bone - Bone: deform_collar_back_r_01"!(ComponentPose=@n90, bAlphaBoolEnabled:bool=True, Translation="-0.180,-0.000,4.219", Rotation="15.000,-0.000,0.000", Scale="1,1,1") -> (Pose) {
    alphaInputType=Bool
    rotationMode=BMM_Additive
    rotationSpace=BCS_BoneSpace
    translationMode=BMM_Additive
    translationSpace=BCS_BoneSpace
    alphaBoolBlend {blendInTime=0.3, blendOption=ExpInOut, blendOutTime=0.3}
    alphaScaleBiasClamp {bInterpResult=true}
    boneToModify {boneName=deform_collar_back_r_01}
    rotation=<rot 15 0 0>
    translation {x=-0.18, z=4.219}
   }
   n92 = "Transform (Modify) Bone - Bone: deform_collar_back_01"!(ComponentPose=@n91, Alpha:float=@n40, Translation="0,0,0", Rotation="0,0,0", Scale="1,1,1") -> (Pose) {
    rotationMode=BMM_Additive
    rotationSpace=BCS_BoneSpace
    translationMode=BMM_Additive
    translationSpace=BCS_BoneSpace
    alphaBoolBlend {blendOption=Linear}
    alphaScaleBiasClamp {interpSpeedDecreasing=30}
    boneToModify {boneName=deform_collar_back_01}
   }
   n93 = "Transform (Modify) Bone - Bone: deform_collar_back_r_01"!(ComponentPose=@n92, Alpha:float=@n40, Translation="0,0,0", Rotation="0,0,0", Scale="1,1,1") -> (Pose) {
    rotationMode=BMM_Additive
    rotationSpace=BCS_BoneSpace
    translationMode=BMM_Additive
    translationSpace=BCS_BoneSpace
    alphaBoolBlend {blendOption=Linear}
    alphaScaleBiasClamp {interpSpeedDecreasing=30}
    boneToModify {boneName=deform_collar_back_r_01}
   }
   n94 = "Transform (Modify) Bone - Bone: deform_collar_back_l_01"!(ComponentPose=@n93, Alpha:float=@n40, Translation="0,0,0", Rotation="0,0,0", Scale="1,1,1") -> (Pose) {
    rotationMode=BMM_Additive
    rotationSpace=BCS_BoneSpace
    translationMode=BMM_Additive
    translationSpace=BCS_BoneSpace
    alphaBoolBlend {blendOption=Linear}
    alphaScaleBiasClamp {interpSpeedDecreasing=30}
    boneToModify {boneName=deform_collar_back_l_01}
   }
   n95 = "Transform (Modify) Bone - Bone: deform_collar_front_r_01"(ComponentPose=@n94, Alpha=@n40, Translation="-1.617,0.000,1.927", Rotation="40.000,-0.000,-0.000", Scale="1,1,1") {
    rotationMode=BMM_Additive
    rotationSpace=BCS_BoneSpace
    translationMode=BMM_Additive
    translationSpace=BCS_BoneSpace
    alphaBoolBlend {blendOption=Linear}
    alphaScaleBiasClamp {interpSpeedDecreasing=1, interpSpeedIncreasing=1}
    boneToModify {boneName=deform_collar_front_r_01}
    rotation=<rot 40 0 0>
    translation {x=-1.617, z=1.927}
   }
   n96 = "Transform (Modify) Bone - Bone: deform_collar_front_l_01"(ComponentPose=@n95, Alpha=@n40, Translation="0,0,0", Rotation="0,0,0", Scale="1,1,1") {
    rotationMode=BMM_Additive
    rotationSpace=BCS_BoneSpace
    translationMode=BMM_Additive
    translationSpace=BCS_BoneSpace
    alphaBoolBlend {blendOption=Linear}
    alphaScaleBiasClamp {interpSpeedDecreasing=30}
    boneToModify {boneName=deform_collar_front_l_01}
   }
   n97 = "Transform (Modify) Bone - Bone: deform_collar_back_01"(ComponentPose=@n96, Alpha=@n39, Translation="0,0,0", Rotation="0,0,0", Scale="1,1,1") {
    rotationMode=BMM_Additive
    rotationSpace=BCS_BoneSpace
    translationMode=BMM_Additive
    translationSpace=BCS_BoneSpace
    alphaBoolBlend {blendOption=Linear}
    alphaScaleBiasClamp {interpSpeedDecreasing=30}
    boneToModify {boneName=deform_collar_back_01}
   }
   n98 = "Transform (Modify) Bone - Bone: deform_collar_back_r_01"(ComponentPose=@n97, Alpha=@n39, Translation="0,0,0", Rotation="0,0,0", Scale="1,1,1") {
    rotationMode=BMM_Additive
    rotationSpace=BCS_BoneSpace
    translationMode=BMM_Additive
    translationSpace=BCS_BoneSpace
    alphaBoolBlend {blendOption=Linear}
    alphaScaleBiasClamp {interpSpeedDecreasing=30}
    boneToModify {boneName=deform_collar_back_r_01}
   }
   n99 = "Transform (Modify) Bone - Bone: deform_collar_back_l_01"(ComponentPose=@n98, Alpha=@n39, Translation="-0.000,0.000,5.936", Rotation="0,0,0", Scale="1,1,1") {
    rotationMode=BMM_Additive
    rotationSpace=BCS_BoneSpace
    translationMode=BMM_Additive
    translationSpace=BCS_BoneSpace
    alphaBoolBlend {blendOption=Linear}
    alphaScaleBiasClamp {bInterpResult=true, interpSpeedDecreasing=20, interpSpeedIncreasing=20}
    boneToModify {boneName=deform_collar_back_l_01}
    translation {z=5.936}
   }
   n100 = "Transform (Modify) Bone - Bone: deform_collar_front_r_01"(ComponentPose=@n99, Alpha=@n39, Translation="0,0,0", Rotation="0,0,0", Scale="1,1,1") {
    rotationMode=BMM_Additive
    rotationSpace=BCS_BoneSpace
    translationMode=BMM_Additive
    translationSpace=BCS_BoneSpace
    alphaBoolBlend {blendOption=Linear}
    alphaScaleBiasClamp {interpSpeedDecreasing=30}
    boneToModify {boneName=deform_collar_front_r_01}
   }
   n101 = "Transform (Modify) Bone - Bone: deform_collar_front_l_01"(ComponentPose=@n100, Alpha=@n39, Translation="-2.738,6.694,3.205", Rotation="15.000,0.000,0.000", Scale="1,1,1") {
    rotationMode=BMM_Additive
    rotationSpace=BCS_BoneSpace
    translationMode=BMM_Additive
    translationSpace=BCS_BoneSpace
    alphaBoolBlend {blendOption=Linear}
    alphaScaleBiasClamp {bInterpResult=true, interpSpeedDecreasing=20, interpSpeedIncreasing=20}
    boneToModify {boneName=deform_collar_front_l_01}
    rotation=<rot 15 0 0.0001>
    translation=<-2.738 6.694 3.205>
   }
   Save cached pose 'PostCollar'(Pose=Component To Local(ComponentPose=@n101)) {cacheName=PostCollar}
   n102 = RigidBody(
    ComponentPose=Local To Component(LocalPose=Use cached pose 'Post-skirt setup'())
    Alpha=Get notInVehicle()
    OverrideWorldGravity=Get GravityOverride()
    ExternalForce=Get skydiving vector force()
   ) {
    bClampLinearTranslationLimitToRefPose=true
    bOverrideWorldGravity=true
    overridePhysicsAsset=&/F_MED_DarkNinja_Physics
    simulationSpace=BaseBoneSpace
    alphaBoolBlend {blendOption=Linear}
    baseBoneRef {boneName=thigh_l}
    componentAppliedLinearAccClamp=<1250 2000 2000>
    componentLinearAccScale=<1 0.65 0.65>
    componentLinearVelScale=<0.2 0.35 0.35>
    overrideWorldGravity {z=-980}
    simSpaceSettings {maxAngularAcceleration=1000, maxAngularVelocity=3, maxLinearAcceleration=250, maxLinearVelocity=500, velocityScaleZ=0, worldAlpha=1}
   }
   n103 = Get CustomIdleIsPlaying()
   n104 = "Transform (Modify) Bone - Bone: dyn_skirtFront_01_Base"(ComponentPose=@n102, Alpha=@n103, Translation="0.5,0.5,0", Rotation="0,0,0", Scale="1,1,1") {
    rotationMode=BMM_Additive
    scaleMode=BMM_Additive
    translationMode=BMM_Additive
    alphaBoolBlend {blendOption=Linear}
    boneToModify {boneName=dyn_skirtFront_01_Base}
    translation=<0.5 0.5>
   }
   n105 = "Transform (Modify) Bone - Bone: dyn_skirtFrontSIde_01_Base"(ComponentPose=@n104, Alpha=@n103, Translation="-3.37,0.501,0", Rotation="6,0,0", Scale="1,1,1") {
    rotationMode=BMM_Additive
    scaleMode=BMM_Additive
    translationMode=BMM_Additive
    alphaBoolBlend {blendOption=Linear}
    boneToModify {boneName=dyn_skirtFrontSIde_01_Base}
    rotation {pitch=6}
    translation=<-3.37 0.501>
   }
   n106 = "Transform (Modify) Bone - Bone: dyn_skirtSIde_01_Base"(ComponentPose=@n105, Alpha=@n103, Translation="-1,0,0", Rotation="8,0,0", Scale="1,1,1") {
    rotationMode=BMM_Additive
    scaleMode=BMM_Additive
    translationMode=BMM_Additive
    alphaBoolBlend {blendOption=Linear}
    boneToModify {boneName=dyn_skirtSIde_01_Base}
    rotation {pitch=8}
    translation {x=-1}
   }
   n107 = "Transform (Modify) Bone - Bone: dyn_skirtSIde_03"(ComponentPose=@n106, Alpha=@n103, Translation="0,0,0", Rotation="-6,0,0", Scale="1,1,1") {
    rotationMode=BMM_Additive
    scaleMode=BMM_Additive
    translationMode=BMM_Additive
    alphaBoolBlend {blendOption=Linear}
    boneToModify {boneName=dyn_skirtSIde_03}
    rotation {pitch=-6}
   }
   n108 = "Transform (Modify) Bone - Bone: dyn_SkirtPelvis_01_MSTR"(ComponentPose=@n107, bAlphaBoolEnabled=True, Translation="0,0,0", Rotation="14.199,-48.453,-31.029", Scale="1,1,1") {
    alphaInputType=Bool
    rotationMode=BMM_Additive
    rotationSpace=BCS_BoneSpace
    translationMode=BMM_Additive
    translationSpace=BCS_BoneSpace
    alphaBoolBlend {blendInTime=0.3, blendOption=ExpInOut, blendOutTime=0.3}
    alphaScaleBiasClamp {bInterpResult=true}
    boneToModify {boneName=dyn_SkirtPelvis_01_MSTR}
    rotation=<rot 14.199 -48.4529 -31.029>
   }
   n109 = "Transform (Modify) Bone - Bone: dyn_SkirtThigh_01_MSTR"(ComponentPose=@n108, bAlphaBoolEnabled=True, Translation="0,0,0", Rotation="42.289,-19.905,-18.110", Scale="1,1,1") {
    alphaInputType=Bool
    rotationMode=BMM_Additive
    rotationSpace=BCS_BoneSpace
    translationMode=BMM_Additive
    translationSpace=BCS_BoneSpace
    alphaBoolBlend {blendInTime=0.3, blendOption=ExpInOut, blendOutTime=0.3}
    alphaScaleBiasClamp {bInterpResult=true}
    boneToModify {boneName=dyn_SkirtThigh_01_MSTR}
    rotation=<rot 42.289 -19.905 -18.11>
   }
   n110 = "Transform (Modify) Bone - Bone: dyn_skirtBack_03"(ComponentPose=@n109, bAlphaBoolEnabled=True, Translation="0,0,0", Rotation="-14.367,69.420,3.443", Scale="1,1,1") {
    alphaInputType=Bool
    rotationMode=BMM_Additive
    rotationSpace=BCS_BoneSpace
    translationMode=BMM_Additive
    translationSpace=BCS_BoneSpace
    alphaBoolBlend {blendInTime=0.3, blendOption=ExpInOut, blendOutTime=0.3}
    alphaScaleBiasClamp {bInterpResult=true}
    boneToModify {boneName=dyn_skirtBack_03}
    rotation=<rot -14.367 69.42 3.4429>
   }
   Save cached pose 'rigids_cache'(Pose=Component To Local(ComponentPose=@n110)) {cacheName=rigids_cache}
   n111 = Get Movement to Left or Right Alpha() -> (Movement to Left or Right Alpha:double)
   n112 = "Transform (Modify) Bone - Bone: dyn_SkirtThigh_01_MSTR"!(
    ComponentPose=Local To Component(LocalPose=Use cached pose 'PostCollar'())
    Alpha:float=@n111
    Translation="2,0,0"
    Rotation="35,0,0"
    Scale="1,1,1"
   ) -> (Pose) {
    rotationMode=BMM_Additive
    translationMode=BMM_Additive
    translationSpace=BCS_BoneSpace
    alphaBoolBlend {blendOption=Linear}
    alphaScaleBiasClamp {bInterpResult=true, interpSpeedDecreasing=2, interpSpeedIncreasing=2}
    boneToModify {boneName=dyn_SkirtThigh_01_MSTR}
    rotation {pitch=35}
    translation {x=2}
   }
   n113 = "Transform (Modify) Bone - Bone: dyn_SkirtPelvis_01_MSTR"!(ComponentPose=@n112, Alpha:float=@n111, Translation="3,0,0", Rotation="15,0,0", Scale="1,1,1") -> (Pose) {
    rotationMode=BMM_Additive
    translationMode=BMM_Additive
    alphaBoolBlend {blendOption=Linear}
    alphaScaleBiasClamp {bInterpResult=true, interpSpeedDecreasing=2, interpSpeedIncreasing=2}
    boneToModify {boneName=dyn_SkirtPelvis_01_MSTR}
    rotation {pitch=15}
    translation {x=3}
   }
   n114 = "Transform (Modify) Bone - Bone: dyn_SkirtThigh_01_MSTR"!(ComponentPose=@n113, bAlphaBoolEnabled:bool=True, Translation="0,0,2", Rotation="15,0,0", Scale="1,1,1") -> (Pose) {
    alphaInputType=Bool
    rotationMode=BMM_Additive
    rotationSpace=BCS_BoneSpace
    translationMode=BMM_Additive
    translationSpace=BCS_BoneSpace
    alphaBoolBlend {blendInTime=0.3, blendOutTime=0.3}
    boneToModify {boneName=dyn_SkirtThigh_01_MSTR}
    rotation {pitch=15}
    translation {z=2}
   }
   n115 = "Transform (Modify) Bone - Bone: dyn_SkirtPelvis_01_MSTR"!(ComponentPose=@n114, bAlphaBoolEnabled:bool=True, Translation="0,0,2", Rotation="10,0,0", Scale="1,1,1") -> (Pose) {
    alphaInputType=Bool
    rotationMode=BMM_Additive
    rotationSpace=BCS_BoneSpace
    translationMode=BMM_Additive
    translationSpace=BCS_BoneSpace
    alphaBoolBlend {blendInTime=0.3, blendOutTime=0.3}
    boneToModify {boneName=dyn_SkirtPelvis_01_MSTR}
    rotation {pitch=10}
    translation {z=2}
   }
   Save cached pose 'Post-skirt setup'(Pose=Component To Local(ComponentPose=@n115)) {cacheName=Post-skirt setup}
   Get MeshToCopy() -> (MeshToCopy:Skeletal Mesh Component Object Reference)
  ]}
 EventGraph {nodes [
   Delay(Duration=0.125)
   > ResetDynamics(InTeleportType=TeleportPhysics)
   > ResetDynamics(InTeleportType=ResetPhysics)
   Event BlueprintUpdateAnimation()
   > n1 = Sequence!(execute) -> (then_0, then_1, then_2:Exec, then_3:Exec, then_4:Exec, then_5:Exec, then_6:Exec)
   > n2 = Gravity Override State Function DEPRECATED(CharacterPartAnimInstance=Self-Reference(), GameTimeInSeconds=GetGameTimeInSeconds(), GravityOverrideParams=Get Gravity Override Params())
   > Set GravityOverride(GravityOverride=@n2)
   n3 = Get bIsInVehicleSeat() -> (bIsInVehicleSeat:bool)
   > Set notInVehicle(notInVehicle:double=float - float(B="To Float (Boolean)"(InBool=@n3)) {is=title K2Node_CallFunction}) -> (Output_Get:double)
   > Set crouchAlpha(crouchAlpha:double="To Float (Boolean)"(InBool=Get bIsCrouching())) -> (Output_Get:double)
   n4 = Get PawnSpeed2D()
   > Set speedAlpha(speedAlpha:double="To Float (Boolean)"(InBool="float >= float"(A=@n4, B=405) {is=title K2Node_CallFunction})) -> (Output_Get:double)
   > n5 = Branch(Condition=Get bIsSkyDiving())
   > n6 = Branch(Condition=Get bIsSkydiveFloating())
   n7 = RandomFloatInRangeFromStream!(Stream:"Random Stream Structure (by ref)"=Get random_stream() -> (random_stream:Random Stream Structure), Min:float=-18000, Max:float=18000) -> (ReturnValue:float)
   n8 = MakeVector(X=@n7, Y=@n7)
   > Set skydiving vector force(skydiving vector force:Vector=@n8) -> (Output_Get:Vector)
   n9 = Branch(execute=@n6.else, Condition=Get bIsSkydiveDiveMode())
   n10 = Get Direction Between Sockets!(
    Component:Skeletal Mesh Component Object Reference=GetOwnerSkeletalMeshComponent()
    SocketOrBoneNameFrom:Name=head
    SocketOrBoneNameTo:Name=pelvis
   ) -> (ReturnValue:Vector) {is=title K2Node_CallFunction}
   > Set skydiving vector force(skydiving vector force=vector * vector(A:Vector=vector * vector(A=@n8, B=@n10), B:Vector="0.15,0.15,0.15") -> (ReturnValue:Vector) {is=title K2Node_CommutativeAssociativeBinaryOperator})
   Set skydiving vector force(execute=@n9.else, skydiving vector force=@n8)
   Set skydiving vector force(execute=@n5.else, skydiving vector force="0, 0, 0")
   n11 = Branch(execute=@n1.then_4, Condition=Get bIsPlayingEmote())
   > Set R Arm Raise Alpha(R Arm Raise Alpha:float=0) -> (Output_Get:float)
   Set R Arm Raise Alpha(execute=@n11.else, R Arm Raise Alpha=Get R Arm Raise Alpha())
   n12 = Get Skeletal Mesh() -> (Skeletal Mesh:Skeletal Mesh Component Object Reference)
   n13 = Set R Arm Raise Alpha(
    execute=@n1.then_3
    R Arm Raise Alpha=float + float(
     A=GetAlphaFromReferenceTransformDelta(SkeletalMeshComponent=@n12, BoneName=upperarm_r, BaseName=spine_05, InputRange=(X=60,Y=105), TwistAxis="0,0,1")
     B=GetAlphaFromReferenceTransformDelta(SkeletalMeshComponent=@n12, BoneName=upperarm_r, BaseName=spine_05, InputRange=(X=-180,Y=-60), OutputRange=(X=1,Y=0), TwistAxis="0,0,1")
    )
   )
   > Set Skydiving R Arm Raise Alpha(Skydiving R Arm Raise Alpha:float=float * float(A=@n13, B:bool=Get bIsSkyDiving()) {is=title K2Node_PromotableOperator}) -> (Output_Get:float)
   > Set L Arm Raise Alpha(L Arm Raise Alpha:float=float + float(
     A=GetAlphaFromReferenceTransformDelta(SkeletalMeshComponent=@n12, BoneName=upperarm_l, BaseName=spine_05, InputRange=(X=60,Y=105), TwistAxis="0,0,1")
     B=GetAlphaFromReferenceTransformDelta(SkeletalMeshComponent=@n12, BoneName=upperarm_l, BaseName=spine_05, InputRange=(X=-180,Y=-105), OutputRange=(X=1,Y=0), TwistAxis="0,0,1")
    )) -> (Output_Get:float)
   > Set Spine rotation to the Left while Targeting Alpha(Spine rotation to the Left while Targeting Alpha:double=float * float(
     A=GetAlphaFromReferenceTransformDelta(SkeletalMeshComponent=@n12, BoneName=spine_05, BaseName=pelvis, DeltaComponentType=EulerX, InputRange=(X=-10,Y=7), TwistAxis="0,0,1")
     B=Get bIsTargeting()
    )) -> (Output_Get:double)
   > Set Spine rotation to the Right while Targeting Alpha(Spine rotation to the Right while Targeting Alpha:double=float * float(
     A=GetAlphaFromReferenceTransformDelta(SkeletalMeshComponent=@n12, BoneName=spine_05, BaseName=pelvis, DeltaComponentType=EulerX, InputRange=(X=-10,Y=-45), TwistAxis="0,0,1")
     B=Get bIsTargeting()
    )) -> (Output_Get:double)
   n14 = GetAlphaFromReferenceTransformDelta(SkeletalMeshComponent=@n12, BoneName=head, BaseName=spine_05, DeltaComponentType=EulerX, InputRange=(X=0,Y=-45), TwistAxis="0,0,1")
   > Set Melee Head turn to the right Alpha(Melee Head turn to the right Alpha:double=float * float(A:bool=Get bIsPlayingMeleeAnim(), B:double=@n14)) -> (Output_Get:double)
   > Set Head turn to the right Alpha(Head turn to the right Alpha:double=@n14) -> (Output_Get:double)
   n15 = GetAlphaFromReferenceTransformDelta(SkeletalMeshComponent=@n12, BoneName=head, BaseName=spine_05, DeltaComponentType=EulerX, InputRange=(X=10,Y=45), TwistAxis="0,0,1")
   > Set Melee Head turn to the left Alpha(Melee Head turn to the left Alpha:double=float * float(A=Get bIsPlayingMeleeAnim(), B=@n15)) -> (Output_Get:double)
   > Set Head turn to the left Alpha(Head turn to the left Alpha:double=@n15) -> (Output_Get:double)
   > Set Head look Up Alpha(Head look Up Alpha:double=GetAlphaFromReferenceTransformDelta(SkeletalMeshComponent=@n12, BoneName=head, BaseName=spine_05, InputRange=(X=0,Y=-45), TwistAxis="0,0,1")) -> (Output_Get:double)
   > Set Head Tilt To Right Alpha(Head Tilt To Right Alpha:double=GetAlphaFromReferenceTransformDelta(SkeletalMeshComponent=@n12, BoneName=head, BaseName=spine_05, DeltaComponentType=EulerY, InputRange=(X=0,Y=30), TwistAxis="0,0,1")) -> (Output_Get:double)
   n16 = TryGetPawnOwner()
   n17 = GetVelocity!(self:Actor Object Reference=@n16) -> (ReturnValue:Vector)
   n18 = Dot Product!(A:Vector=GetActorRightVector!(self:Actor Object Reference=@n16) -> (ReturnValue:Vector), B:Vector=@n17) -> (ReturnValue:double)
   n19 = Set Movement to Right Alpha(execute=@n1.then_5, Movement to Right Alpha:double=MapRangeClamped(Value=@n18, InRangeA=0, InRangeB=230)) -> (Output_Get:double)
   > Set Skydiving Movement to Right Alpha(Skydiving Movement to Right Alpha:double=float * float(A:double=@n19, B:bool=Get bIsSkyDiving())) -> (Output_Get:double)
   > n20 = Set Movement to Left Alpha(Movement to Left Alpha:double=MapRangeClamped(Value=@n18, InRangeA=0, InRangeB=-230)) -> (Output_Get:double)
   n21 = Get bIsSkyDiving()
   > Set Skydiving Movement to Left Alpha(Skydiving Movement to Left Alpha:double=float * float(A=@n20, B=@n21)) -> (Output_Get:double)
   > Set Movement to Left or Right Alpha(Movement to Left or Right Alpha:double=float * float(A=float + float(A=@n20, B=@n19), B=NOT Boolean(A=@n21))) -> (Output_Get:double)
   n22 = Dot Product(A=@n17, B=GetActorForwardVector!(self:Actor Object Reference=@n16) -> (ReturnValue:Vector))
   n23 = Set Movement Forward or Backward Alpha(execute=@n1.then_6, Movement Forward or Backward Alpha:double=MapRangeClamped(Value="Absolute (Float)"(A=@n22), InRangeA=0, InRangeB=300)) -> (Output_Get:double)
   > Set Skydiving Movement Forward or Backward Alpha(Skydiving Movement Forward or Backward Alpha:double=float * float(A=@n23, B=Get bIsSkyDiving())) -> (Output_Get:double)
   n24 = "integer > integer"(A=Get CurrentLOD(), B=1) {is=title K2Node_CallFunction}
   n25 = NOT Boolean(A=Get bRigidBodyEnabled())
   Set blendPoseByBoolValue(execute=@n1.then_2, blendPoseByBoolValue:bool=OR Boolean(A=OR Boolean(A=@n25, B=@n24), B=Get bIsInVehicle() -> (bIsInVehicle:bool))) -> (Output_Get:bool)
   > Set Is Skydiving Floating or Dropping(Is Skydiving Floating or Dropping:bool=OR Boolean(A=Get bIsSkydiveFloating(), B=Get bIsSkydiveDiveMode())) -> (Output_Get:bool)
   Branch(execute=@n1.then_1, Condition=Get bIsInFrontEnd())
   > Set CustomIdleIsPlaying(CustomIdleIsPlaying:float=GetCurveValue(self=GetOwnerFrontEndAnimInstance(), CurveName=Frontend_Idle_Bypass)) -> (Output_Get:float)
   Event BlueprintInitializeAnimation()
   > Delay(Duration=0.125)
   > ResetDynamics(InTeleportType=TeleportPhysics)
   > ResetDynamics(InTeleportType=ResetPhysics)
   Event OnCharacterCustomizationLoaded()
   > Set Skeletal Mesh(Skeletal Mesh:Skeletal Mesh Component Object Reference=GetOwnerSkeletalMeshComponent()) -> (Output_Get:Skeletal Mesh Component Object Reference)
   n26 = Sequence!(execute) -> (then_0, then_1, then_2)
   n27 = Get bIsInVehicleSeat()
   > Set notInVehicle(notInVehicle=float - float(B="To Float (Boolean)"(InBool=@n27)))
   > Set crouchAlpha(crouchAlpha="To Float (Boolean)"(InBool=Get bIsCrouching()))
   n28 = Get PawnSpeed2D()
   > Set speedAlpha(speedAlpha="To Float (Boolean)"(InBool="float >= float"(A=@n28, B=405)))
   > n29 = Branch(Condition=Get bIsSkyDiving())
   > n30 = Branch(Condition=Get bIsSkydiveFloating())
   n31 = RandomFloatInRangeFromStream(Stream=Get random_stream(), Min=-18000, Max=18000)
   n32 = MakeVector(X=@n31, Y=@n31)
   > Set skydiving vector force(skydiving vector force=@n32)
   n33 = Branch(execute=@n30.else, Condition=Get bIsSkydiveDiveMode())
   n34 = Get Direction Between Sockets(Component=GetOwnerSkeletalMeshComponent(), SocketOrBoneNameFrom=head, SocketOrBoneNameTo=pelvis)
   > Set skydiving vector force(skydiving vector force=vector * vector(A=vector * vector(A=@n32, B=@n34), B="0.15,0.15,0.15"))
   Set skydiving vector force(execute=@n33.else, skydiving vector force=@n32)
   Set skydiving vector force(execute=@n29.else, skydiving vector force="0, 0, 0")
   Branch(execute=@n26.then_1, Condition=Get bIsInFrontEnd())
   > Set CustomIdleIsPlaying(CustomIdleIsPlaying=GetCurveValue(self=GetOwnerFrontEndAnimInstance(), CurveName=Frontend_Idle_Bypass))
   n35 = NOT Boolean(A=Get bRigidBodyEnabled())
   n36 = "integer > integer"(A=Get CurrentLOD(), B=1)
   Set blendPoseByBoolValue(execute=@n26.then_2, blendPoseByBoolValue=OR Boolean(A=OR Boolean(A=@n35, B=@n36), B=Get bIsInVehicle()))
  ]}
]
variables [
 blendPoseByBoolValue {type=boolean}
 crouchAlpha {type=number}
 CustomIdleIsPlaying {type=number}
 Gravity Override Params {type=GravityOverrideParamsStruct}
 GravityOverride {type=Vector}
 Head look Up Alpha {type=number}
 Head Tilt To Right Alpha {type=number}
 Head turn to the left Alpha {type=number}
 Head turn to the right Alpha {type=number}
 Is Skydiving Floating or Dropping {type=boolean}
 L Arm Raise Alpha {type=number}
 Melee Head turn to the left Alpha {type=number}
 Melee Head turn to the right Alpha {type=number}
 MeshToCopy {type=SkeletalMeshComponent}
 Movement Forward or Backward Alpha {type=number}
 Movement to Left Alpha {type=number}
 Movement to Left or Right Alpha {type=number}
 Movement to Right Alpha {type=number}
 notInVehicle {type=number}
 R Arm Raise Alpha {type=number}
 random_stream {type=RandomStream, default {initialSeed=0, seed=0}}
 Skeletal Mesh {type=SkeletalMeshComponent}
 Skydiving Movement Forward or Backward Alpha {type=number}
 Skydiving Movement to Left Alpha {type=number}
 Skydiving Movement to Right Alpha {type=number}
 Skydiving R Arm Raise Alpha {type=number}
 skydiving vector force {type=Vector}
 speedAlpha {type=number}
 Spine rotation to the Left while Targeting Alpha {type=number}
 Spine rotation to the Right while Targeting Alpha {type=number}
]
